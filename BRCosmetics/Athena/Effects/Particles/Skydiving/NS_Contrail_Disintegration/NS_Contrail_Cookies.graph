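# Niagara skydiving contrail cosmetic that trails cookie bits from the character and bursts them at recorded death positions.
# Trail emitter places mesh bits via skeletal mesh sockets or bones, with rate and lifetime driven by glide lean.
# Burst emitters fire on a system death flag gated by lean, adding time-of-day-tinted dust sprites.
# Meshes use weighted random selection and camera-based scaling; older Bites bursts are disabled.
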NiagaraSystem /BRCosmetics/Athena/Effects/Particles/Skydiving/NS_Contrail_Disintegration/NS_Contrail_Cookies
dependencies {
 usedDataInterfaces [
  &NiagaraDataInterfaceCurve
  /Script/Niagara.NiagaraDataInterfaceArrayDistributionInt
  &NiagaraDataInterfaceArrayInt32
  &NiagaraDataInterfaceCamera
  &NiagaraDataInterfaceVector2DCurve
  &NiagaraDataInterfaceColorCurve
  &NiagaraDataInterfaceSocketReader
  &NiagaraDataInterfaceSkeletalMesh
  &NiagaraDataInterfaceArrayPosition
 ]
 usedDynamicInputs [
  &MakeQuatFromAxisAngle
  &MaskFloatBySpawnGroup
  &Vector2DFromFloat
 ]
 usedModules [
  &self:SystemSpawnScript.SystemScriptSource.SystemScriptGraph.NiagaraNodeAssignment_0.SetVariables_#1
  &self:SystemSpawnScript.SystemScriptSource.SystemScriptGraph.NiagaraNodeAssignment_2.SetVariables_#2
  &SystemState
  &Timeline
  /Game/Effects/Niagara/Modules/Contrails/FN_Contrail_Setup
  &FN_Integrator
  &self:SystemSpawnScript.SystemScriptSource.SystemScriptGraph.NiagaraNodeAssignment_1.SetVariables_#3
  &self:ReadDeath
  &self:SystemSpawnScript.SystemScriptSource.SystemScriptGraph.NiagaraNodeAssignment_3.SetVariables_#4
  &BlendSpaceHelper
 ]
}
emitters {
 Bites_Burst {
  bEnabled=false
  emitterUpdateScript [
   &EmitterState(Life Cycle Mode=::NewEnumerator1, Loop Behavior=::NewEnumerator1, Loop Duration Mode=::NewEnumerator1, UseLoopDelay=false)
   &SpawnPerFrame(
    Spawn:NiagaraBool=&Boolean_AndOperation(Bool1=$System.DeathSpawn, Bool2=&SetBoolByFloatComparison(A=$User.ParaGlideLeanAlpha, B=0.1))
    Spawn Count:NiagaraInt32=&RandomRangeInt(Minimum=6, Maximum=9)
    Use Spawn Probability=false
   )
   &SpawnPerFrame(
    Spawn:NiagaraBool=&Boolean_AndOperation(Bool1=$System.DeathSpawn, Bool2=&SetBoolByFloatComparison(A=$User.ParaGlideLeanAlpha, B=0.1))
    Spawn Count:NiagaraInt32=&RandomRangeInt(Minimum=6, Maximum=9)
    Use Spawn Probability=false
   ) {enabled=false, module=SpawnPerFrame001}
  ]
  particleSpawnScript [
   /SetVariables(Particles.MeshIndex=&SelectIntFromArray(
     Direct Array Index=/Niagara/DynamicInputs/Arrays/SelectIntFromWeightedDistributionArray(
      Array Sampling Mode:ENiagara_LifetimeMode=::NewEnumerator1
      Direct Array Index:NiagaraInt32=?
      Direct Array Mode:ENiagara_WrapClamp=::NewEnumerator0
      Evaluation Type:ENiagaraRandomnessEvaluation=::NewEnumerator0
      Recalculate Random Each Loop:NiagaraBool=false
      Weighted Distribution Int Array:NiagaraDataInterfaceArrayDistributionInt=DataInterface() {propertyValues {ArrayData [{value=0, weight=0.4}, {value=1, weight=0.4}, {value=0, weight=0.1}, {value=0, weight=0.1}]}}
      Randomness Mode:ENiagaraRandomnessMode=::NewEnumerator0
      Random Seed:NiagaraInt32=0
      Fixed Random Seed:NiagaraBool=false
     )
     Int Selection Array=DataInterface() {propertyValues {IntData [0,1,2,3]}}
    )) {isSetParameters=true, module=SetVariables_#5}
   &PartitionParticles()
   &InitializeParticle(
    Color Maximum=<rgba 0.9688 0.9688 0.9688 1>
    Color Minimum=<rgba 0.4125 0.4125 0.4125 1>
    Color Mode=::NewEnumerator2
    Lifetime Max=0.5
    Lifetime Min=0.3
    Lifetime Mode=::NewEnumerator1
    Mesh Scale Mode=::NewEnumerator3
    Mesh Uniform Scale=&MergeHelper_FloatSwitch(
     0=&RandomRangeFloat(Minimum=0.3, Maximum=0.5)
     1=&RandomRangeFloat(Minimum=0.3, Maximum=0.5)
     2=&RandomRangeFloat(Minimum=0.25, Maximum=0.4)
     3=&RandomRangeFloat(Minimum=0.25, Maximum=0.4)
     Index=$Particles.MeshIndex
     Num=::NewEnumerator2
    )
    Sprite Size Mode=::NewEnumerator3
    Uniform Sprite Size=20
    UsePositionOffset=false
    Write Material Random=true
   )
   &ShapeLocation(
    Cylinder Height=2
    Cylinder Radius=6
    MaskSpawnGroup=false
    Override Seed=false
    Shape Origin=$System.DeathPosition
    Shape Primitive=::NewEnumerator1
    Surface Only=false
    UseOwnerScale=true
   )
   &AddVelocity(
    Module.Use Falloff Distance=false
    Origin Offset=&AddVectorToPosition(Position=$System.DeathPosition, Vector=<0 0 0>)
    Override Seed=false
    Velocity Mode=::NewEnumerator1
    Velocity Speed=&RandomRangeFloat(Minimum=30, Maximum=100)
   ) {module=AddVelocity001}
   &InitialMeshOrientation(Use Rotation Vector=false)
   &ApplyInitialForces()
  ]
  particleUpdateScript [
   &ParticleState()
   &ScaleMeshSize() {module=ScaleMeshSize002}
   &UpdateMeshOrientation(
    Rotation Rate=&RandomRangeFloat(Minimum=-8, Maximum=8)
    Rotation Vector=&RandomRangeVector(Minimum=<-1 -1 -1>, Maximum=<1 1 1>)
   )
   &Drag(Drag=&RandomRangeFloat(Minimum=7, Maximum=12), Use Linear Drag=true, Use Rotational Drag=true)
   &GravityForce(Gravity=&Lerp_Vector(
     A=&Lerp_Vector(
      A=<0 0 0>
      B=&RandomRangeVector(Minimum=<0 0 1000>, Maximum=<0 0 1800>)
      Alpha=&FloatFromCurve(FloatCurve=DataInterface() {propertyValues {Curve {defaultValue=0, keys [{tangentMode=RCTM_User, time=0, value=0}, {tangentMode=RCTM_User, time=1, value=1}]}}})
     )
     B=&Lerp_Vector(
      A=<0 0 0>
      B=&RandomRangeVector(Minimum=<0 0 2000>, Maximum=<0 0 3200>)
      Alpha=&FloatFromCurve(FloatCurve=DataInterface() {propertyValues {Curve {defaultValue=0, keys [{tangentMode=RCTM_User, time=0, value=0}, {tangentMode=RCTM_User, time=1, value=1}]}}})
     )
     Alpha=$User.ParaGlideLeanAlpha
    ))
   &GravityForce(Gravity=&Lerp_Vector(
     A=&RandomRangeVector(Minimum=<0 0 1000>, Maximum=<0 0 1800>)
     B=&RandomRangeVector(Minimum=<0 0 2000>, Maximum=<0 0 3200>)
     Alpha=$User.ParaGlideLeanAlpha
    )) {enabled=false, module=GravityForce001}
   &SolveForcesAndVelocity(Clamp Velocity=false, Limit Acceleration=false)
   &ScaleMeshSize(Scale Factor=&VectorFromFloat(Value=&FloatFromCurve(
      FloatCurve=DataInterface() {propertyValues {Curve {keys [
          {arriveTangent=41.8211, interpMode=RCIM_Cubic, leaveTangent=41.8211, tangentMode=RCTM_User, time=0, value=0}
          {interpMode=RCIM_Cubic, tangentMode=RCTM_User, time=0.02, value=1}
          {interpMode=RCIM_Cubic, tangentMode=RCTM_User, time=0.5, value=1}
          {arriveTangent=-1.4032, interpMode=RCIM_Cubic, leaveTangent=-1.4032, tangentMode=RCTM_User, time=1, value=0.5}
         ]}}}
      Scale Curve=/Game/Effects/Niagara/Modules/Math/FN_CameraDistance(
       0Dist:NiagaraFloat=40
       1Dist:NiagaraFloat=90
       Camera Query:NiagaraDataInterfaceCamera=DataInterface()
       Position:NiagaraPosition=$Particles.Position
      )
     )))
   &ScaleMeshSize(Scale Factor=&VectorFromFloat(Value=&Lerp_Float(
      A=1.1
      B=0.7
      Alpha=&Multiply_Float(
       A=&DotProduct(
        Vector A=&NormalizeVector(
         Vector To Normalize=&ReturnCameraVector(Camera Vectors=::NewEnumerator1, Coordinate Space=::World)
         Safe Normalize=true
         Safe Fallback Threshold=0
         Safe Fallback Vector=<0 0 0.001>
        )
        Vector B=&NormalizeVector(Vector To Normalize=$Engine.Owner.SystemZAxis, Safe Normalize=true, Safe Fallback Threshold=0, Safe Fallback Vector=<0 0 0.001>)
       )
       B=-1
      )
     ))) {module=ScaleMeshSize001}
  ]
  rendererClasses [&NiagaraMeshRendererProperties]
  renderers [{rendererClass=&NiagaraMeshRendererProperties, rendererIndex=0}]
 }
 Bites_Burst001 {
  bEnabled=false
  emitterUpdateScript [
   &EmitterState(Life Cycle Mode=::NewEnumerator1, Loop Behavior=::NewEnumerator1, Loop Duration Mode=::NewEnumerator1, UseLoopDelay=false)
   &SpawnPerFrame(
    Spawn:NiagaraBool=&Boolean_AndOperation(Bool1=$System.DeathSpawn, Bool2=&SetBoolByFloatComparison(A=$User.ParaGlideLeanAlpha, B=0.1))
    Spawn Count:NiagaraInt32=&RandomRangeInt(Minimum=6, Maximum=9)
    Use Spawn Probability=false
   )
   &SpawnPerFrame(
    Spawn:NiagaraBool=&Boolean_AndOperation(Bool1=$System.DeathSpawn, Bool2=&SetBoolByFloatComparison(A=$User.ParaGlideLeanAlpha, B=0.1))
    Spawn Count:NiagaraInt32=&RandomRangeInt(Minimum=6, Maximum=9)
    Use Spawn Probability=false
   ) {enabled=false, module=SpawnPerFrame001}
  ]
  particleSpawnScript [
   /SetVariables(Particles.MeshIndex=&SelectIntFromArray(
     Direct Array Index=&SelectIntFromWeightedDistributionArray(
      Array Sampling Mode:ENiagara_LifetimeMode=::NewEnumerator1
      Direct Array Index:NiagaraInt32=?
      Direct Array Mode:ENiagara_WrapClamp=::NewEnumerator0
      Evaluation Type:ENiagaraRandomnessEvaluation=::NewEnumerator0
      Recalculate Random Each Loop:NiagaraBool=false
      Weighted Distribution Int Array:NiagaraDataInterfaceArrayDistributionInt=DataInterface() {propertyValues {ArrayData [{value=0, weight=0.4}, {value=1, weight=0.4}, {value=0, weight=0.1}, {value=0, weight=0.1}]}}
      Randomness Mode:ENiagaraRandomnessMode=::NewEnumerator0
      Random Seed:NiagaraInt32=0
      Fixed Random Seed:NiagaraBool=false
     )
     Int Selection Array=DataInterface() {propertyValues {IntData [0,1,2,3]}}
    )) {isSetParameters=true, module=SetVariables_#5}
   &PartitionParticles()
   &InitializeParticle(
    Color Maximum=<rgba 0.9688 0.9688 0.9688 1>
    Color Minimum=<rgba 0.4125 0.4125 0.4125 1>
    Color Mode=::NewEnumerator2
    Lifetime Max=0.5
    Lifetime Min=0.3
    Lifetime Mode=::NewEnumerator1
    Mesh Scale Mode=::NewEnumerator3
    Mesh Uniform Scale=&MergeHelper_FloatSwitch(
     0=&RandomRangeFloat(Minimum=0.3, Maximum=0.5)
     1=&RandomRangeFloat(Minimum=0.3, Maximum=0.5)
     2=&RandomRangeFloat(Minimum=0.25, Maximum=0.4)
     3=&RandomRangeFloat(Minimum=0.25, Maximum=0.4)
     Index=$Particles.MeshIndex
     Num=::NewEnumerator2
    )
    Sprite Size Mode=::NewEnumerator3
    Uniform Sprite Size=20
    UsePositionOffset=false
    Write Material Random=true
   )
   &ShapeLocation(
    Cylinder Height=2
    Cylinder Radius=6
    MaskSpawnGroup=false
    Override Seed=false
    Shape Origin=$System.DeathPosition
    Shape Primitive=::NewEnumerator1
    Surface Only=false
    UseOwnerScale=true
   )
   &AddVelocity(
    Module.Use Falloff Distance=false
    Origin Offset=&AddVectorToPosition(Position=$System.DeathPosition, Vector=<0 0 0>)
    Override Seed=false
    Velocity Mode=::NewEnumerator1
    Velocity Speed=&RandomRangeFloat(Minimum=30, Maximum=100)
   ) {module=AddVelocity001}
   &InitialMeshOrientation(Use Rotation Vector=false)
   &ApplyInitialForces()
  ]
  particleUpdateScript [
   &ParticleState()
   &ScaleMeshSize() {module=ScaleMeshSize002}
   &UpdateMeshOrientation(
    Rotation Rate=&RandomRangeFloat(Minimum=-8, Maximum=8)
    Rotation Vector=&RandomRangeVector(Minimum=<-1 -1 -1>, Maximum=<1 1 1>)
   )
   &Drag(Drag=&RandomRangeFloat(Minimum=7, Maximum=12), Use Linear Drag=true, Use Rotational Drag=true)
   &GravityForce(Gravity=&Lerp_Vector(
     A=&Lerp_Vector(
      A=<0 0 0>
      B=&RandomRangeVector(Minimum=<0 0 1000>, Maximum=<0 0 1800>)
      Alpha=&FloatFromCurve(FloatCurve=DataInterface() {propertyValues {Curve {defaultValue=0, keys [{tangentMode=RCTM_User, time=0, value=0}, {tangentMode=RCTM_User, time=1, value=1}]}}})
     )
     B=&Lerp_Vector(
      A=<0 0 0>
      B=&RandomRangeVector(Minimum=<0 0 2000>, Maximum=<0 0 3200>)
      Alpha=&FloatFromCurve(FloatCurve=DataInterface() {propertyValues {Curve {defaultValue=0, keys [{tangentMode=RCTM_User, time=0, value=0}, {tangentMode=RCTM_User, time=1, value=1}]}}})
     )
     Alpha=$User.ParaGlideLeanAlpha
    ))
   &GravityForce(Gravity=&Lerp_Vector(
     A=&RandomRangeVector(Minimum=<0 0 1000>, Maximum=<0 0 1800>)
     B=&RandomRangeVector(Minimum=<0 0 2000>, Maximum=<0 0 3200>)
     Alpha=$User.ParaGlideLeanAlpha
    )) {enabled=false, module=GravityForce001}
   &SolveForcesAndVelocity(Clamp Velocity=false, Limit Acceleration=false)
   &ScaleMeshSize(Scale Factor=&VectorFromFloat(Value=&FloatFromCurve(
      FloatCurve=DataInterface() {propertyValues {Curve {keys [
          {arriveTangent=41.8211, interpMode=RCIM_Cubic, leaveTangent=41.8211, tangentMode=RCTM_User, time=0, value=0}
          {interpMode=RCIM_Cubic, tangentMode=RCTM_User, time=0.02, value=1}
          {interpMode=RCIM_Cubic, tangentMode=RCTM_User, time=0.5, value=1}
          {arriveTangent=-1.4032, interpMode=RCIM_Cubic, leaveTangent=-1.4032, tangentMode=RCTM_User, time=1, value=0.5}
         ]}}}
      Scale Curve=&FN_CameraDistance(
       0Dist:NiagaraFloat=40
       1Dist:NiagaraFloat=90
       Camera Query:NiagaraDataInterfaceCamera=DataInterface()
       Position:NiagaraPosition=$Particles.Position
      )
     )))
   &ScaleMeshSize(Scale Factor=&VectorFromFloat(Value=&Lerp_Float(
      A=1.1
      B=0.7
      Alpha=&Multiply_Float(
       A=&DotProduct(
        Vector A=&NormalizeVector(
         Vector To Normalize=&ReturnCameraVector(Camera Vectors=::NewEnumerator1, Coordinate Space=::World)
         Safe Normalize=true
         Safe Fallback Threshold=0
         Safe Fallback Vector=<0 0 0.001>
        )
        Vector B=&NormalizeVector(Vector To Normalize=$Engine.Owner.SystemZAxis, Safe Normalize=true, Safe Fallback Threshold=0, Safe Fallback Vector=<0 0 0.001>)
       )
       B=-1
      )
     ))) {module=ScaleMeshSize001}
  ]
  rendererClasses [&NiagaraMeshRendererProperties]
  renderers [{rendererClass=&NiagaraMeshRendererProperties, rendererIndex=0}]
 }
 Bites_Burst003 {
  bEnabled=false
  emitterUpdateScript [
   &EmitterState(Life Cycle Mode=::NewEnumerator1, Loop Behavior=::NewEnumerator1, Loop Duration Mode=::NewEnumerator1, UseLoopDelay=false)
   &SpawnPerFrame(
    Spawn:NiagaraBool=&Boolean_AndOperation(Bool1=$System.DeathSpawn, Bool2=&SetBoolByFloatComparison(A=$User.ParaGlideLeanAlpha, B=0.1))
    Spawn Count:NiagaraInt32=&RandomRangeInt(Minimum=10, Maximum=15)
    Use Spawn Probability=false
   )
   &SpawnPerFrame(
    Spawn:NiagaraBool=&Boolean_AndOperation(Bool1=$System.DeathSpawn, Bool2=&SetBoolByFloatComparison(A=$User.ParaGlideLeanAlpha, B=0.1))
    Spawn Count:NiagaraInt32=&RandomRangeInt(Minimum=6, Maximum=9)
    Use Spawn Probability=false
   ) {enabled=false, module=SpawnPerFrame001}
   /Game/Effects/Niagara/Modules/Logic/UpdateValueStateChange() {enabled=false}
  ]
  particleSpawnScript [
   /SetVariables(Particles.MeshIndex=&SelectIntFromArray(
     Direct Array Index=&SelectIntFromWeightedDistributionArray(
      Array Sampling Mode:ENiagara_LifetimeMode=::NewEnumerator1
      Direct Array Index:NiagaraInt32=?
      Direct Array Mode:ENiagara_WrapClamp=::NewEnumerator0
      Evaluation Type:ENiagaraRandomnessEvaluation=::NewEnumerator0
      Recalculate Random Each Loop:NiagaraBool=false
      Weighted Distribution Int Array:NiagaraDataInterfaceArrayDistributionInt=DataInterface() {propertyValues {ArrayData [{value=0, weight=0.4}, {value=1, weight=0.4}, {value=0, weight=0.1}, {value=0, weight=0.1}]}}
      Randomness Mode:ENiagaraRandomnessMode=::NewEnumerator0
      Random Seed:NiagaraInt32=0
      Fixed Random Seed:NiagaraBool=false
     )
     Int Selection Array=DataInterface() {propertyValues {IntData [0,1,2,3]}}
    )) {isSetParameters=true, module=SetVariables_#5}
   &PartitionParticles()
   &InitializeParticle(
    Color Maximum=<rgba 0.9688 0.9688 0.9688 1>
    Color Minimum=<rgba 0.4125 0.4125 0.4125 1>
    Color Mode=::NewEnumerator2
    Lifetime Max=0.5
    Lifetime Min=0.3
    Lifetime Mode=::NewEnumerator1
    Mesh Scale Mode=::NewEnumerator3
    Mesh Uniform Scale=&MergeHelper_FloatSwitch(
     0=&RandomRangeFloat(Minimum=0.3, Maximum=0.5)
     1=&RandomRangeFloat(Minimum=0.3, Maximum=0.5)
     2=&RandomRangeFloat(Minimum=0.25, Maximum=0.4)
     3=&RandomRangeFloat(Minimum=0.25, Maximum=0.4)
     Index=$Particles.MeshIndex
     Num=::NewEnumerator2
    )
    Sprite Size Mode=::NewEnumerator3
    Uniform Sprite Size=20
    UsePositionOffset=false
    Write Material Random=true
   )
   &ShapeLocation(
    Cylinder Height=2
    Cylinder Radius=6
    MaskSpawnGroup=false
    Override Seed=false
    Shape Origin=$System.DeathPosition
    Shape Primitive=::NewEnumerator1
    Surface Only=false
    UseOwnerScale=true
   )
   &AddVelocity(
    Module.Use Falloff Distance=false
    Origin Offset=&AddVectorToPosition(Position=$System.DeathPosition, Vector=<0 0 0>)
    Override Seed=false
    Velocity Mode=::NewEnumerator1
    Velocity Speed=&RandomRangeFloat(Minimum=30, Maximum=100)
   ) {module=AddVelocity001}
   &InitialMeshOrientation(Use Rotation Vector=false)
   &ApplyInitialForces()
  ]
  particleUpdateScript [
   &ParticleState()
   &ScaleMeshSize() {module=ScaleMeshSize002}
   &UpdateMeshOrientation(
    Rotation Rate=&RandomRangeFloat(Minimum=-8, Maximum=8)
    Rotation Vector=&RandomRangeVector(Minimum=<-1 -1 -1>, Maximum=<1 1 1>)
   )
   &Drag(Drag=&RandomRangeFloat(Minimum=5, Maximum=10), Use Linear Drag=true, Use Rotational Drag=true)
   &GravityForce(Gravity=&Lerp_Vector(
     A=&Lerp_Vector(
      A=<0 0 0>
      B=&RandomRangeVector(Minimum=<0 0 1000>, Maximum=<0 0 1800>)
      Alpha=&FloatFromCurve(
       FloatCurve=DataInterface() {propertyValues {Curve {defaultValue=0, keys [{tangentMode=RCTM_User, time=0, value=0}, {tangentMode=RCTM_User, time=1, value=1}]}}}
       Scale Curve=0.5
      )
     )
     B=&Lerp_Vector(
      A=<0 0 0>
      B=&RandomRangeVector(Minimum=<0 0 2000>, Maximum=<0 0 3200>)
      Alpha=&FloatFromCurve(
       FloatCurve=DataInterface() {propertyValues {Curve {defaultValue=0, keys [{tangentMode=RCTM_User, time=0, value=0}, {tangentMode=RCTM_User, time=1, value=1}]}}}
       Scale Curve=0.5
      )
     )
     Alpha=$User.ParaGlideLeanAlpha
    ))
   &GravityForce(Gravity=&Lerp_Vector(
     A=&RandomRangeVector(Minimum=<0 0 1000>, Maximum=<0 0 1800>)
     B=&RandomRangeVector(Minimum=<0 0 2000>, Maximum=<0 0 3200>)
     Alpha=$User.ParaGlideLeanAlpha
    )) {enabled=false, module=GravityForce001}
   &SolveForcesAndVelocity(Clamp Velocity=false, Limit Acceleration=false)
   &ScaleMeshSize(Scale Factor=&VectorFromFloat(Value=&FloatFromCurve(
      FloatCurve=DataInterface() {propertyValues {Curve {keys [
          {arriveTangent=41.8211, interpMode=RCIM_Cubic, leaveTangent=41.8211, tangentMode=RCTM_User, time=0, value=0}
          {interpMode=RCIM_Cubic, tangentMode=RCTM_User, time=0.02, value=1}
          {interpMode=RCIM_Cubic, tangentMode=RCTM_User, time=0.5, value=1}
          {arriveTangent=-1.4032, interpMode=RCIM_Cubic, leaveTangent=-1.4032, tangentMode=RCTM_User, time=1, value=0.5}
         ]}}}
      Scale Curve=&FN_CameraDistance(
       0Dist:NiagaraFloat=40
       1Dist:NiagaraFloat=90
       Camera Query:NiagaraDataInterfaceCamera=DataInterface()
       Position:NiagaraPosition=$Particles.Position
      )
     )))
   &ScaleMeshSize(Scale Factor=&VectorFromFloat(Value=&Lerp_Float(
      A=1.1
      B=0.7
      Alpha=&Multiply_Float(
       A=&DotProduct(
        Vector A=&NormalizeVector(
         Vector To Normalize=&ReturnCameraVector(Camera Vectors=::NewEnumerator1, Coordinate Space=::World)
         Safe Normalize=true
         Safe Fallback Threshold=0
         Safe Fallback Vector=<0 0 0.001>
        )
        Vector B=&NormalizeVector(Vector To Normalize=$Engine.Owner.SystemZAxis, Safe Normalize=true, Safe Fallback Threshold=0, Safe Fallback Vector=<0 0 0.001>)
       )
       B=-1
      )
     ))) {module=ScaleMeshSize001}
  ]
  rendererClasses [&NiagaraMeshRendererProperties]
  renderers [{rendererClass=&NiagaraMeshRendererProperties, rendererIndex=0}]
 }
 Bits_Bursts {
  emitterSpawnScript [/SetVariables(Emitter.Max Traversal Speed=300) {isSetParameters=true}]
  emitterUpdateScript [
   &EmitterState(Life Cycle Mode=::NewEnumerator1, Loop Behavior=::NewEnumerator1, Loop Duration Mode=::NewEnumerator1, UseLoopDelay=false)
   &SpawnPerFrame(
    Spawn:NiagaraBool=&Boolean_AndOperation(Bool1=$System.DeathSpawn, Bool2=&SetBoolByFloatComparison(A=$User.ParaGlideLeanAlpha, B=0.1))
    Spawn Count:NiagaraInt32=&RandomRangeInt(Minimum=10, Maximum=15)
    Use Spawn Probability=false
   )
   &SpawnPerFrame(
    Spawn:NiagaraBool=&Boolean_AndOperation(Bool1=$System.DeathSpawn, Bool2=&SetBoolByFloatComparison(A=$User.ParaGlideLeanAlpha, B=0.1))
    Spawn Count:NiagaraInt32=&RandomRangeInt(Minimum=4, Maximum=7)
    Spawn Group=1
    Use Spawn Probability=false
   ) {module=SpawnPerFrame002}
   &SpawnPerFrame(
    Spawn:NiagaraBool=&Boolean_AndOperation(Bool1=$System.DeathSpawn, Bool2=&SetBoolByFloatComparison(A=$User.ParaGlideLeanAlpha, B=0.1))
    Spawn Count:NiagaraInt32=&RandomRangeInt(Minimum=6, Maximum=9)
    Use Spawn Probability=false
   ) {enabled=false, module=SpawnPerFrame001}
   &UpdateValueStateChange() {enabled=false}
  ]
  particleSpawnScript [
   /SetVariables(
    Particles.MeshIndex=&SelectIntFromArray(
     Direct Array Index=&SelectIntFromWeightedDistributionArray(
      Array Sampling Mode:ENiagara_LifetimeMode=::NewEnumerator1
      Direct Array Index:NiagaraInt32=?
      Direct Array Mode:ENiagara_WrapClamp=::NewEnumerator0
      Evaluation Type:ENiagaraRandomnessEvaluation=::NewEnumerator0
      Recalculate Random Each Loop:NiagaraBool=false
      Weighted Distribution Int Array:NiagaraDataInterfaceArrayDistributionInt=DataInterface() {propertyValues {ArrayData [{value=0, weight=0.4}, {value=1, weight=0.4}, {value=0, weight=0.1}, {value=0, weight=0.1}]}}
      Randomness Mode:ENiagaraRandomnessMode=::NewEnumerator0
      Random Seed:NiagaraInt32=0
      Fixed Random Seed:NiagaraBool=false
     )
     Int Selection Array=DataInterface() {propertyValues {IntData [0,1,2,3]}}
    )
    Particles.VisibilityTag=&SpawnGroup()
   ) {isSetParameters=true, module=SetVariables_#5}
   &PartitionParticles()
   &InitializeParticle(
    Color=&MergeHelper_ColorSwitch(
     0=&RandomRangeLinearColor(Minimum=<rgba 0.4125 0.4125 0.4125 1>, Maximum=<rgba 0.9688 0.9688 0.9688 1>)
     1=&RandomRangeLinearColor(Minimum=<rgba 0.478 0.3791 0.2199 0.2>, Maximum=<rgba 0.781 0.7271 0.5498 0.4>)
     2=?
     3=?
     Index=0
     Num=::NewEnumerator0
    )
    Lifetime=&MergeHelper_FloatSwitch(
     0=&RandomRangeFloat(Minimum=0.15, Maximum=0.3)
     1=&RandomRangeFloat(Minimum=0.3, Maximum=0.5)
     2=?
     3=?
     Index=$Particles.VisibilityTag
     Num=::NewEnumerator0
    )
    Mesh Scale Mode=::NewEnumerator3
    Mesh Uniform Scale=&MergeHelper_FloatSwitch(
     0=&RandomRangeFloat(Minimum=0.3, Maximum=0.5)
     1=&RandomRangeFloat(Minimum=0.3, Maximum=0.5)
     2=&RandomRangeFloat(Minimum=0.25, Maximum=0.4)
     3=&RandomRangeFloat(Minimum=0.25, Maximum=0.4)
     Index=$Particles.MeshIndex
     Num=::NewEnumerator2
    )
    Sprite Size Mode=::NewEnumerator3
    Uniform Sprite Size=&MergeHelper_FloatSwitch(
     0=20
     1=&RandomRangeFloat(Minimum=10, Maximum=15)
     2=?
     3=?
     Index=$Particles.VisibilityTag
     Num=::NewEnumerator0
    )
    UsePositionOffset=false
    Write Material Random=true
   )
   &ShapeLocation(
    Cylinder Height=2
    Cylinder Radius=6
    MaskSpawnGroup=false
    Override Seed=false
    Shape Origin=$System.DeathPosition
    Shape Primitive=::NewEnumerator1
    Surface Only=false
    UseOwnerScale=true
   )
   &AddVelocity(
    Module.Use Falloff Distance=false
    Origin Offset=&AddVectorToPosition(Position=$System.DeathPosition, Vector=<0 0 0>)
    Override Seed=false
    Velocity Mode=::NewEnumerator1
    Velocity Speed=&MergeHelper_FloatSwitch(
     0=&RandomRangeFloat(Minimum=30, Maximum=100)
     1=&RandomRangeFloat(Minimum=35, Maximum=80)
     2=?
     3=?
     Index=$Particles.VisibilityTag
     Num=::NewEnumerator0
    )
   ) {module=AddVelocity001}
   &InitialMeshOrientation(Use Rotation Vector=false)
   &ApplyInitialForces()
  ]
  particleUpdateScript [
   &ParticleState()
   &ScaleMeshSize() {module=ScaleMeshSize002}
   &UpdateMeshOrientation(
    Rotation Rate=&RandomRangeFloat(Minimum=-8, Maximum=8)
    Rotation Vector=&RandomRangeVector(Minimum=<-1 -1 -1>, Maximum=<1 1 1>)
   )
   &Drag(
    Drag=&MergeHelper_FloatSwitch(
     0=&RandomRangeFloat(Minimum=5, Maximum=10)
     1=&FloatFromCurve(
      FloatCurve=DataInterface() {propertyValues {Curve {
         defaultValue=1
         keys [{interpMode=RCIM_Cubic, tangentMode=RCTM_User, time=0, value=0.75}, {interpMode=RCIM_Cubic, tangentMode=RCTM_User, time=0.65, value=1}]
        }}}
      Scale Curve=&Lerp_Float(
       A=20
       B=8
       Alpha=&RemapRange(
        Input Value To Remap=&VectorLength(Vector=$User.OwnerLocalVelocity)
        Input Min=0
        Input Max=$Emitter.Max Traversal Speed
        Output Min=0
        Output Max=1
       )
      )
     )
     2=?
     3=?
     Index=$Particles.VisibilityTag
     Num=::NewEnumerator0
    )
    Use Linear Drag=true
    Use Rotational Drag=true
   )
   &GravityForce(Gravity=&MergeHelper_VectorSwitch(
     0=&Lerp_Vector(
      A=&Lerp_Vector(
       A=<0 0 0>
       B=&RandomRangeVector(Minimum=<0 0 1000>, Maximum=<0 0 1800>)
       Alpha=&FloatFromCurve(
        FloatCurve=DataInterface() {propertyValues {Curve {defaultValue=0, keys [{tangentMode=RCTM_User, time=0, value=0}, {tangentMode=RCTM_User, time=1, value=1}]}}}
        Scale Curve=0.5
       )
      )
      B=&Lerp_Vector(
       A=<0 0 0>
       B=&RandomRangeVector(Minimum=<0 0 2000>, Maximum=<0 0 3200>)
       Alpha=&FloatFromCurve(
        FloatCurve=DataInterface() {propertyValues {Curve {defaultValue=0, keys [{tangentMode=RCTM_User, time=0, value=0}, {tangentMode=RCTM_User, time=1, value=1}]}}}
        Scale Curve=0.5
       )
      )
      Alpha=$User.ParaGlideLeanAlpha
     )
     1=<0 0 1500>
     2=?
     3=?
     Index=$Particles.VisibilityTag
     Num=::NewEnumerator0
    ))
   &ScaleSpriteSize(Uniform Curve Sprite Scale=DataInterface() {propertyValues {Curve {keys [
        {arriveTangent=0.8753, interpMode=RCIM_Cubic, leaveTangent=0.8753, tangentMode=RCTM_User, time=0, value=1}
        {arriveTangent=0.0238, interpMode=RCIM_Cubic, leaveTangent=0.0238, tangentMode=RCTM_User, time=1, value=1.3}
       ]}}})
   &GravityForce(Gravity=&Lerp_Vector(
     A=&RandomRangeVector(Minimum=<0 0 1000>, Maximum=<0 0 1800>)
     B=&RandomRangeVector(Minimum=<0 0 2000>, Maximum=<0 0 3200>)
     Alpha=$User.ParaGlideLeanAlpha
    )) {enabled=false, module=GravityForce001}
   &SolveForcesAndVelocity(Clamp Velocity=false, Limit Acceleration=false)
   &ScaleMeshSize(Scale Factor=&VectorFromFloat(Value=&FloatFromCurve(
      FloatCurve=DataInterface() {propertyValues {Curve {keys [
          {arriveTangent=41.8211, interpMode=RCIM_Cubic, leaveTangent=41.8211, tangentMode=RCTM_User, time=0, value=0}
          {interpMode=RCIM_Cubic, tangentMode=RCTM_User, time=0.02, value=1}
          {interpMode=RCIM_Cubic, tangentMode=RCTM_User, time=0.5, value=1}
          {arriveTangent=-1.4032, interpMode=RCIM_Cubic, leaveTangent=-1.4032, tangentMode=RCTM_User, time=1, value=0}
         ]}}}
      Scale Curve=&FN_CameraDistance(
       0Dist:NiagaraFloat=40
       1Dist:NiagaraFloat=90
       Camera Query:NiagaraDataInterfaceCamera=DataInterface()
       Position:NiagaraPosition=$Particles.Position
      )
     )))
   &ScaleMeshSize(Scale Factor=&VectorFromFloat(Value=&Lerp_Float(
      A=1.1
      B=0.7
      Alpha=&Multiply_Float(
       A=&DotProduct(
        Vector A=&NormalizeVector(
         Vector To Normalize=&ReturnCameraVector(Camera Vectors=::NewEnumerator1, Coordinate Space=::World)
         Safe Normalize=true
         Safe Fallback Threshold=0
         Safe Fallback Vector=<0 0 0.001>
        )
        Vector B=&NormalizeVector(Vector To Normalize=$Engine.Owner.SystemZAxis, Safe Normalize=true, Safe Fallback Threshold=0, Safe Fallback Vector=<0 0 0.001>)
       )
       B=-1
      )
     ))) {module=ScaleMeshSize001}
   /SetVariables(
    Particles.ColorDust:LinearColor=&MultiplyLinearColor(
     Linear Color=&RandomRangeLinearColor(Minimum=<rgba 0.5 0.2354 0.0464 1>, Maximum=<rgba 0.5 0.311 0.0464 1>)
     Scale Factor=&MakeVector4_FromLinearColor(LinearColor=&MakeLinearColorFromVectorAndFloat(
       "Vector (RGB)"=&MakeVectorFromLinearColorRGB(Color=&FNTimeOfDayColor(
         Night Color=&RandomRangeLinearColor(Minimum=<rgba 0.4138 0.5201 0.6661 1>, Maximum=<rgba 0.781 0.781 0.781 1>)
         Morning Color=&RandomRangeLinearColor(Minimum=<rgba 0.6882 0.6227 0.9531 1>, Maximum=<rgba 0.6063 0.7457 0.9346 1>)
         Day Color=&RandomRangeLinearColor(Minimum=<rgba 0.478 0.4265 0.4121 0.8>, Maximum=<rgba 0.781 0.718 0.6943 0.9>)
         Evening Color=&RandomRangeLinearColor(Minimum=<rgba 0.478 0.478 0.478 1>, Maximum=<rgba 0.781 0.781 0.781 1>)
        ))
       "Float (Alpha)"=&FloatFromCurve(
        FloatCurve=DataInterface() {propertyValues {Curve {keys [
            {interpMode=RCIM_Cubic, tangentMode=RCTM_User, time=0, value=0}
            {arriveTangentWeight=0.0754, interpMode=RCIM_Cubic, leaveTangentWeight=0.2297, tangentMode=RCTM_User, time=0.1684, value=1}
            {arriveTangent=-0.5014, interpMode=RCIM_Cubic, leaveTangent=-0.5014, tangentMode=RCTM_User, time=0.4045, value=1}
            {arriveTangent=-2.9749, interpMode=RCIM_Cubic, leaveTangent=-2.9749, tangentMode=RCTM_User, time=0.9, value=0}
           ]}}}
        Scale Curve=&Multiply_Float(
         A=&RemapRange(
          Input Value To Remap=&Abs_Float(Float:NiagaraFloat=&DotProduct(Vector A=&ReturnCameraVector(Camera Vectors=::NewEnumerator1, Coordinate Space=::Simulation), Vector B=<0 0 1>))
          Input Min=0
          Input Max=0.8
          Output Min=0.6
          Output Max=0.3
         )
         B=&RemapRange(
          Input Value To Remap=&MakeFloatFromVector(Vector=Linked() {linkedVariable {name=Particles.Initial.Position, type=NiagaraPosition}}, Channel=::NewEnumerator2)
          Input Min=125
          Input Max=250
          Output Min=0.3
          Output Max=0.05
         )
        )
       )
      ))
    )
    Particles.DynamicParam_Dust:Vector4f=&MakeVector4(
     X=&Add_Float(
      A=&RandomRangeFloat(Minimum=0, Maximum=1)
      B=&FloatFromCurve(
       FloatCurve=DataInterface() {propertyValues {Curve {keys [
           <at 0 0 RCIM_Cubic>
           {arriveTangent=1.5947, interpMode=RCIM_Cubic, leaveTangent=1.5947, tangentMode=RCTM_User, time=0.2557, value=0.5257}
           <at 1 1 RCIM_Cubic>
          ]}}}
       CurveIndex=$Particles.Age
       Scale Curve=0.25
      )
     )
     Y=0
     Z=0
     W=0
    )
   ) {isSetParameters=true}
   /SetVariables(
    Particles.ColorDust:LinearColor=&LerpLinearColors(
     StartColor=<rgba 1 0.01 0 1>
     EndColor=<rgba 0 0.1858 1 1>
     LerpFactor=&RemapRange(
      Input Value To Remap=&MakeFloatFromVector(Vector=Linked() {linkedVariable {name=Particles.Initial.Position, type=NiagaraPosition}}, Channel=::NewEnumerator2)
      Input Min=125
      Input Max=250
      Output Min=0
      Output Max=1
     )
    )
    Particles.DynamicParam_Dust:Vector4f=&MakeVector4(
     X=&Add_Float(
      A=&RandomRangeFloat(Minimum=0, Maximum=1)
      B=&FloatFromCurve(
       FloatCurve=DataInterface() {propertyValues {Curve {keys [
           <at 0 0 RCIM_Cubic>
           {arriveTangent=1.5947, interpMode=RCIM_Cubic, leaveTangent=1.5947, tangentMode=RCTM_User, time=0.2557, value=0.5257}
           <at 1 1 RCIM_Cubic>
          ]}}}
       CurveIndex=$Particles.Age
       Scale Curve=0.25
      )
     )
     Y=0
     Z=0
     W=0
    )
   ) {enabled=false, isSetParameters=true}
   &BlendSpace(Position Coordinate Space=1, Rotation Contribution=0) {enabled=false}
  ]
  rendererClasses [&NiagaraMeshRendererProperties, &NiagaraSpriteRendererProperties]
  renderers [
   {rendererClass=&NiagaraMeshRendererProperties, rendererIndex=0}
   {rendererClass=&NiagaraSpriteRendererProperties, rendererIndex=1}
   {rendererClass=&NiagaraSpriteRendererProperties, rendererIndex=2}
  ]
 }
 Bits_Bursts001 {
  emitterUpdateScript [
   &EmitterState(Life Cycle Mode=::NewEnumerator1, Loop Behavior=::NewEnumerator1, Loop Duration Mode=::NewEnumerator1, UseLoopDelay=false)
   &SpawnPerFrame(
    Spawn:NiagaraBool=&Boolean_AndOperation(Bool1=$System.DeathSpawn, Bool2=&SetBoolByFloatComparison(A=$User.ParaGlideLeanAlpha, B=0.1))
    Spawn Count:NiagaraInt32=&RandomRangeInt(Minimum=10, Maximum=15)
    Use Spawn Probability=false
   )
   &SpawnPerFrame(
    Spawn:NiagaraBool=&Boolean_AndOperation(Bool1=$System.DeathSpawn, Bool2=&SetBoolByFloatComparison(A=$User.ParaGlideLeanAlpha, B=0.1))
    Spawn Count:NiagaraInt32=&RandomRangeInt(Minimum=6, Maximum=9)
    Use Spawn Probability=false
   ) {enabled=false, module=SpawnPerFrame001}
   &UpdateValueStateChange() {enabled=false}
  ]
  particleSpawnScript [
   /SetVariables(Particles.MeshIndex=&SelectIntFromArray(
     Direct Array Index=&SelectIntFromWeightedDistributionArray(
      Array Sampling Mode:ENiagara_LifetimeMode=::NewEnumerator1
      Direct Array Index:NiagaraInt32=?
      Direct Array Mode:ENiagara_WrapClamp=::NewEnumerator0
      Evaluation Type:ENiagaraRandomnessEvaluation=::NewEnumerator0
      Recalculate Random Each Loop:NiagaraBool=false
      Weighted Distribution Int Array:NiagaraDataInterfaceArrayDistributionInt=DataInterface() {propertyValues {ArrayData [{value=0, weight=0.4}, {value=1, weight=0.4}, {value=0, weight=0.1}, {value=0, weight=0.1}]}}
      Randomness Mode:ENiagaraRandomnessMode=::NewEnumerator0
      Random Seed:NiagaraInt32=0
      Fixed Random Seed:NiagaraBool=false
     )
     Int Selection Array=DataInterface() {propertyValues {IntData [0,1,2,3]}}
    )) {isSetParameters=true, module=SetVariables_#5}
   &PartitionParticles()
   &InitializeParticle(
    Color Maximum=<rgba 0.9688 0.9688 0.9688 1>
    Color Minimum=<rgba 0.4125 0.4125 0.4125 1>
    Color Mode=::NewEnumerator2
    Lifetime Max=0.3
    Lifetime Min=0.15
    Lifetime Mode=::NewEnumerator1
    Mesh Scale Mode=::NewEnumerator3
    Mesh Uniform Scale=&MergeHelper_FloatSwitch(
     0=&RandomRangeFloat(Minimum=0.3, Maximum=0.5)
     1=&RandomRangeFloat(Minimum=0.3, Maximum=0.5)
     2=&RandomRangeFloat(Minimum=0.25, Maximum=0.4)
     3=&RandomRangeFloat(Minimum=0.25, Maximum=0.4)
     Index=$Particles.MeshIndex
     Num=::NewEnumerator2
    )
    Sprite Size Mode=::NewEnumerator3
    Uniform Sprite Size=20
    UsePositionOffset=false
    Write Material Random=true
   )
   &ShapeLocation(
    Cylinder Height=2
    Cylinder Radius=6
    MaskSpawnGroup=false
    Override Seed=false
    Shape Origin=$System.DeathPosition
    Shape Primitive=::NewEnumerator1
    Surface Only=false
    UseOwnerScale=true
   )
   &AddVelocity(
    Module.Use Falloff Distance=false
    Origin Offset=&AddVectorToPosition(Position=$System.DeathPosition, Vector=<0 0 0>)
    Override Seed=false
    Velocity Mode=::NewEnumerator1
    Velocity Speed=&RandomRangeFloat(Minimum=30, Maximum=100)
   ) {module=AddVelocity001}
   &InitialMeshOrientation(Use Rotation Vector=false)
   &ApplyInitialForces()
  ]
  particleUpdateScript [
   &ParticleState()
   &ScaleMeshSize() {module=ScaleMeshSize002}
   &UpdateMeshOrientation(
    Rotation Rate=&RandomRangeFloat(Minimum=-8, Maximum=8)
    Rotation Vector=&RandomRangeVector(Minimum=<-1 -1 -1>, Maximum=<1 1 1>)
   )
   &Drag(Drag=&RandomRangeFloat(Minimum=5, Maximum=10), Use Linear Drag=true, Use Rotational Drag=true)
   &GravityForce(Gravity=&Lerp_Vector(
     A=&Lerp_Vector(
      A=<0 0 0>
      B=&RandomRangeVector(Minimum=<0 0 1000>, Maximum=<0 0 1800>)
      Alpha=&FloatFromCurve(
       FloatCurve=DataInterface() {propertyValues {Curve {defaultValue=0, keys [{tangentMode=RCTM_User, time=0, value=0}, {tangentMode=RCTM_User, time=1, value=1}]}}}
       Scale Curve=0.5
      )
     )
     B=&Lerp_Vector(
      A=<0 0 0>
      B=&RandomRangeVector(Minimum=<0 0 2000>, Maximum=<0 0 3200>)
      Alpha=&FloatFromCurve(
       FloatCurve=DataInterface() {propertyValues {Curve {defaultValue=0, keys [{tangentMode=RCTM_User, time=0, value=0}, {tangentMode=RCTM_User, time=1, value=1}]}}}
       Scale Curve=0.5
      )
     )
     Alpha=$User.ParaGlideLeanAlpha
    ))
   &GravityForce(Gravity=&Lerp_Vector(
     A=&RandomRangeVector(Minimum=<0 0 1000>, Maximum=<0 0 1800>)
     B=&RandomRangeVector(Minimum=<0 0 2000>, Maximum=<0 0 3200>)
     Alpha=$User.ParaGlideLeanAlpha
    )) {enabled=false, module=GravityForce001}
   &SolveForcesAndVelocity(Clamp Velocity=false, Limit Acceleration=false)
   &ScaleMeshSize(Scale Factor=&VectorFromFloat(Value=&FloatFromCurve(
      FloatCurve=DataInterface() {propertyValues {Curve {keys [
          {arriveTangent=41.8211, interpMode=RCIM_Cubic, leaveTangent=41.8211, tangentMode=RCTM_User, time=0, value=0}
          {interpMode=RCIM_Cubic, tangentMode=RCTM_User, time=0.02, value=1}
          {interpMode=RCIM_Cubic, tangentMode=RCTM_User, time=0.5, value=1}
          {arriveTangent=-1.4032, interpMode=RCIM_Cubic, leaveTangent=-1.4032, tangentMode=RCTM_User, time=1, value=0}
         ]}}}
      Scale Curve=&FN_CameraDistance(
       0Dist:NiagaraFloat=40
       1Dist:NiagaraFloat=90
       Camera Query:NiagaraDataInterfaceCamera=DataInterface()
       Position:NiagaraPosition=$Particles.Position
      )
     )))
   &ScaleMeshSize(Scale Factor=&VectorFromFloat(Value=&Lerp_Float(
      A=1.1
      B=0.7
      Alpha=&Multiply_Float(
       A=&DotProduct(
        Vector A=&NormalizeVector(
         Vector To Normalize=&ReturnCameraVector(Camera Vectors=::NewEnumerator1, Coordinate Space=::World)
         Safe Normalize=true
         Safe Fallback Threshold=0
         Safe Fallback Vector=<0 0 0.001>
        )
        Vector B=&NormalizeVector(Vector To Normalize=$Engine.Owner.SystemZAxis, Safe Normalize=true, Safe Fallback Threshold=0, Safe Fallback Vector=<0 0 0.001>)
       )
       B=-1
      )
     ))) {module=ScaleMeshSize001}
  ]
  rendererClasses [&NiagaraMeshRendererProperties]
  renderers [{rendererClass=&NiagaraMeshRendererProperties, rendererIndex=0}]
 }
 Bits_Trail {
  emitterUpdateScript [
   &EmitterState(Life Cycle Mode=::NewEnumerator1, Loop Behavior=::NewEnumerator1, Loop Duration Mode=::NewEnumerator1, UseLoopDelay=false)
   &SpawnRate(SpawnRate=&Lerp_Float(A=50, B=120, Alpha=$User.ParaGlideLeanAlpha), Use Spawn Probability=false)
  ]
  particleSpawnScript [
   /SetVariables(Particles.MeshIndex=&SelectIntFromArray(
     Direct Array Index=&SelectIntFromWeightedDistributionArray(
      Array Sampling Mode:ENiagara_LifetimeMode=::NewEnumerator1
      Direct Array Index:NiagaraInt32=?
      Direct Array Mode:ENiagara_WrapClamp=::NewEnumerator0
      Evaluation Type:ENiagaraRandomnessEvaluation=::NewEnumerator0
      Recalculate Random Each Loop:NiagaraBool=false
      Weighted Distribution Int Array:NiagaraDataInterfaceArrayDistributionInt=DataInterface() {propertyValues {ArrayData [{value=0, weight=0.4}, {value=1, weight=0.4}, {value=0, weight=0.1}, {value=0, weight=0.1}]}}
      Randomness Mode:ENiagaraRandomnessMode=::NewEnumerator0
      Random Seed:NiagaraInt32=0
      Fixed Random Seed:NiagaraBool=false
     )
     Int Selection Array=DataInterface() {propertyValues {IntData [0,1,2,3]}}
    )) {isSetParameters=true, module=SetVariables_#6}
   &PartitionParticles(Partition Mode:ENiagara_PartitionMode=::NewEnumerator5)
   &InitializeParticle(
    Color Maximum=<rgba 0.9688 0.9688 0.9688 1>
    Color Minimum=<rgba 0.4125 0.4125 0.4125 1>
    Color Mode=::NewEnumerator2
    Lifetime=&Lerp_Float(
     A=&RandomRangeFloat(Minimum=0.3, Maximum=0.5)
     B=&RandomRangeFloat(Minimum=0.1, Maximum=0.2)
     Alpha=$User.ParaGlideLeanAlpha
    )
    Mesh Scale Mode=::NewEnumerator3
    Mesh Uniform Scale=&MergeHelper_FloatSwitch(
     0=&RandomRangeFloat(Minimum=0.3, Maximum=0.5)
     1=&RandomRangeFloat(Minimum=0.3, Maximum=0.5)
     2=&RandomRangeFloat(Minimum=0.25, Maximum=0.4)
     3=&RandomRangeFloat(Minimum=0.25, Maximum=0.4)
     Index=$Particles.MeshIndex
     Num=::NewEnumerator2
    )
    UsePositionOffset=false
    Write Material Random=true
   )
   &ShapeLocation(
    Disc Coverage=1
    MaskSpawnGroup=false
    Offset=<0 5 5>
    Offset Mode=::NewEnumerator0
    Override Seed=false
    Ring Radius=12
    Shape Primitive=::NewEnumerator4
    UseOwnerScale=true
   )
   &SkeletalMeshLocation(
    Bone / Socket Index=&SelectIntFromArray(Direct Array Index=$Particles.PartitionParticles.Partition, Int Selection Array=DataInterface() {propertyValues {IntData [0,1]}})
    MaskSpawnGroup=false
    Mesh Sampling Type=::NewEnumerator5
    Orientation Sampling=::NewEnumerator1
    Position Sampling=::NewEnumerator1
    Sampled Position Offset:NiagaraPosition=$Particles.ShapeLocation.ShapePosition
    Sampled Position Offset Coordinate Space:ENiagaraSkelMeshTransforms=::NewEnumerator1
    Skeletal Mesh=$System.Skeletal Mesh
    Socket Sampling Mode=::NewEnumerator5
   )
   &AddVelocity(
    Cone Angle=15
    Cone Axis=&Multiply_VectorByFloat(Vector=$System.FN_Contrail_Setup.Direction Vector, Float=-1)
    Module.UseConeFalloff=false
    Module.UseVelDistribution=false
    Override Seed=false
    Velocity Mode=::NewEnumerator2
    Velocity Speed=&Lerp_Float(
     A=&RandomRangeFloat(Minimum=75, Maximum=300)
     B=&RandomRangeFloat(Minimum=300, Maximum=500)
     Alpha=$User.ParaGlideLeanAlpha
    )
   )
   &AddVelocity(
    Cone Angle=15
    Cone Axis=&Multiply_VectorByFloat(Vector=$System.FN_Contrail_Setup.Direction Vector, Float=-1)
    Module.UseConeFalloff=false
    Module.UseVelDistribution=false
    Override Seed=false
    Velocity Mode=::NewEnumerator2
    Velocity Speed=&Lerp_Float(
     A=&RandomRangeFloat(Minimum=225, Maximum=425)
     B=&RandomRangeFloat(Minimum=550, Maximum=700)
     Alpha=$User.ParaGlideLeanAlpha
    )
   ) {enabled=false, module=AddVelocity001}
   &InitialMeshOrientation(Use Rotation Vector=false)
   &ShapeLocation(MaskSpawnGroup=false, Override Seed=false, UseOwnerScale=true) {enabled=false, module=ShapeLocation001}
   &CurlNoiseForce(Bias Noise=false, Noise Strength=40, Pan Noise=false)
   &ApplyInitialForces()
  ]
  particleUpdateScript [
   &ParticleState()
   &ScaleMeshSize(Scale Factor=&VectorFromFloat(Value=&Lerp_Float(A=1, B=0.85, Alpha=$User.ParaGlideLeanAlpha))) {module=ScaleMeshSize002}
   &UpdateMeshOrientation(
    Rotation Rate=&RandomRangeFloat(Minimum=-8, Maximum=8)
    Rotation Vector=&RandomRangeVector(Minimum=<-1 -1 -1>, Maximum=<1 1 1>)
   )
   &Drag(Drag=&RandomRangeFloat(Minimum=0.8, Maximum=1.2), Use Linear Drag=true, Use Rotational Drag=true)
   &SolveForcesAndVelocity(Clamp Velocity=false, Limit Acceleration=false)
   &ScaleColor(
    Scale Alpha=&FloatFromCurve(FloatCurve=DataInterface() {propertyValues {Curve {keys [
         {tangentMode=RCTM_User, time=0, value=0}
         <at 0.0599 1>
         {arriveTangent=-0.0405, interpMode=RCIM_Cubic, leaveTangent=-0.0405, tangentMode=RCTM_User, time=0.7009, value=1}
         {arriveTangent=-8.0089, interpMode=RCIM_Cubic, leaveTangent=-8.0089, tangentMode=RCTM_User, time=1, value=0}
        ]}}})
    ScaleA=true
    ScaleRGB=true
   )
   &ScaleMeshSize(Scale Factor=&VectorFromFloat(Value=&FloatFromCurve(
      FloatCurve=DataInterface() {propertyValues {Curve {keys [
          {arriveTangent=41.8211, interpMode=RCIM_Cubic, leaveTangent=41.8211, tangentMode=RCTM_User, time=0, value=0}
          {interpMode=RCIM_Cubic, tangentMode=RCTM_User, time=0.02, value=1}
          {interpMode=RCIM_Cubic, tangentMode=RCTM_User, time=0.5, value=1}
          {arriveTangent=-1.4032, interpMode=RCIM_Cubic, leaveTangent=-1.4032, tangentMode=RCTM_User, time=1, value=0.5}
         ]}}}
      Scale Curve=&FN_CameraDistance(
       0Dist:NiagaraFloat=40
       1Dist:NiagaraFloat=90
       Camera Query:NiagaraDataInterfaceCamera=DataInterface()
       Position:NiagaraPosition=$Particles.Position
      )
     )))
   &ScaleMeshSize(Scale Factor=&VectorFromFloat(Value=&Lerp_Float(
      A=1
      B=0.9
      Alpha=&Multiply_Float(
       A=&DotProduct(
        Vector A=&NormalizeVector(
         Vector To Normalize=&ReturnCameraVector(Camera Vectors=::NewEnumerator1, Coordinate Space=::World)
         Safe Normalize=true
         Safe Fallback Threshold=0
         Safe Fallback Vector=<0 0 0.001>
        )
        Vector B=&NormalizeVector(Vector To Normalize=$Engine.Owner.SystemZAxis, Safe Normalize=true, Safe Fallback Threshold=0, Safe Fallback Vector=<0 0 0.001>)
       )
       B=-1
      )
     ))) {module=ScaleMeshSize001}
  ]
  rendererClasses [&NiagaraMeshRendererProperties]
  renderers [{rendererClass=&NiagaraMeshRendererProperties, rendererIndex=0}]
 }
 Bits_Trail001 {
  bEnabled=false
  emitterUpdateScript [
   &EmitterState(Life Cycle Mode=::NewEnumerator1, Loop Behavior=::NewEnumerator1, Loop Duration Mode=::NewEnumerator1, UseLoopDelay=false)
   &SpawnRate(SpawnRate=&Lerp_Float(A=50, B=120, Alpha=$User.ParaGlideLeanAlpha), Use Spawn Probability=false)
  ]
  particleSpawnScript [
   /SetVariables(Particles.MeshIndex=&SelectIntFromArray(
     Direct Array Index=&SelectIntFromWeightedDistributionArray(
      Array Sampling Mode:ENiagara_LifetimeMode=::NewEnumerator1
      Direct Array Index:NiagaraInt32=?
      Direct Array Mode:ENiagara_WrapClamp=::NewEnumerator0
      Evaluation Type:ENiagaraRandomnessEvaluation=::NewEnumerator0
      Recalculate Random Each Loop:NiagaraBool=false
      Weighted Distribution Int Array:NiagaraDataInterfaceArrayDistributionInt=DataInterface() {propertyValues {ArrayData [{value=0, weight=0.4}, {value=1, weight=0.4}, {value=0, weight=0.1}, {value=0, weight=0.1}]}}
      Randomness Mode:ENiagaraRandomnessMode=::NewEnumerator0
      Random Seed:NiagaraInt32=0
      Fixed Random Seed:NiagaraBool=false
     )
     Int Selection Array=DataInterface() {propertyValues {IntData [0,1,2,3]}}
    )) {isSetParameters=true, module=SetVariables_#6}
   &PartitionParticles(Partition Mode:ENiagara_PartitionMode=::NewEnumerator5)
   &InitializeParticle(
    Color Maximum=<rgba 0.9688 0.9688 0.9688 1>
    Color Minimum=<rgba 0.4125 0.4125 0.4125 1>
    Color Mode=::NewEnumerator2
    Lifetime=&Lerp_Float(
     A=&RandomRangeFloat(Minimum=0.3, Maximum=0.5)
     B=&RandomRangeFloat(Minimum=0.1, Maximum=0.2)
     Alpha=$User.ParaGlideLeanAlpha
    )
    Mesh Scale Mode=::NewEnumerator3
    Mesh Uniform Scale=&MergeHelper_FloatSwitch(
     0=&RandomRangeFloat(Minimum=0.3, Maximum=0.5)
     1=&RandomRangeFloat(Minimum=0.3, Maximum=0.5)
     2=&RandomRangeFloat(Minimum=0.25, Maximum=0.4)
     3=&RandomRangeFloat(Minimum=0.25, Maximum=0.4)
     Index=$Particles.MeshIndex
     Num=::NewEnumerator2
    )
    UsePositionOffset=false
    Write Material Random=true
   )
   &ShapeLocation(
    Disc Coverage=1
    MaskSpawnGroup=false
    Offset=<0 5 5>
    Offset Mode=::NewEnumerator0
    Override Seed=false
    Ring Radius=12
    Shape Primitive=::NewEnumerator4
    UseOwnerScale=true
   )
   &SkeletalMeshLocation(
    Bone / Socket Index=&SelectIntFromArray(Direct Array Index=$Particles.PartitionParticles.Partition, Int Selection Array=DataInterface() {propertyValues {IntData [0,1]}})
    MaskSpawnGroup=false
    Mesh Sampling Type=::NewEnumerator5
    Orientation Sampling=::NewEnumerator1
    Position Sampling=::NewEnumerator1
    Sampled Position Offset:NiagaraPosition=$Particles.ShapeLocation.ShapePosition
    Sampled Position Offset Coordinate Space:ENiagaraSkelMeshTransforms=::NewEnumerator1
    Skeletal Mesh=$System.Skeletal Mesh
    Socket Sampling Mode=::NewEnumerator5
   )
   &AddVelocity(
    Cone Angle=15
    Cone Axis=&Multiply_VectorByFloat(Vector=$System.FN_Contrail_Setup.Direction Vector, Float=-1)
    Module.UseConeFalloff=false
    Module.UseVelDistribution=false
    Override Seed=false
    Velocity Mode=::NewEnumerator2
    Velocity Speed=&Lerp_Float(
     A=&RandomRangeFloat(Minimum=75, Maximum=300)
     B=&RandomRangeFloat(Minimum=300, Maximum=500)
     Alpha=$User.ParaGlideLeanAlpha
    )
   )
   &AddVelocity(
    Cone Angle=15
    Cone Axis=&Multiply_VectorByFloat(Vector=$System.FN_Contrail_Setup.Direction Vector, Float=-1)
    Module.UseConeFalloff=false
    Module.UseVelDistribution=false
    Override Seed=false
    Velocity Mode=::NewEnumerator2
    Velocity Speed=&Lerp_Float(
     A=&RandomRangeFloat(Minimum=225, Maximum=425)
     B=&RandomRangeFloat(Minimum=550, Maximum=700)
     Alpha=$User.ParaGlideLeanAlpha
    )
   ) {enabled=false, module=AddVelocity001}
   &InitialMeshOrientation(Use Rotation Vector=false)
   &ShapeLocation(MaskSpawnGroup=false, Override Seed=false, UseOwnerScale=true) {enabled=false, module=ShapeLocation001}
   &CurlNoiseForce(Bias Noise=false, Noise Strength=40, Pan Noise=false)
   &ApplyInitialForces()
  ]
  particleUpdateScript [
   &ParticleState()
   &UpdateMeshOrientation(
    Rotation Rate=&RandomRangeFloat(Minimum=-8, Maximum=8)
    Rotation Vector=&RandomRangeVector(Minimum=<-1 -1 -1>, Maximum=<1 1 1>)
   )
   &Drag(Drag=&RandomRangeFloat(Minimum=0.8, Maximum=1.2), Use Linear Drag=true, Use Rotational Drag=true)
   &SolveForcesAndVelocity(Clamp Velocity=false, Limit Acceleration=false)
   &ScaleColor(
    Scale Alpha=&FloatFromCurve(FloatCurve=DataInterface() {propertyValues {Curve {keys [
         {tangentMode=RCTM_User, time=0, value=0}
         <at 0.0599 1>
         {arriveTangent=-0.0405, interpMode=RCIM_Cubic, leaveTangent=-0.0405, tangentMode=RCTM_User, time=0.7009, value=1}
         {arriveTangent=-8.0089, interpMode=RCIM_Cubic, leaveTangent=-8.0089, tangentMode=RCTM_User, time=1, value=0}
        ]}}})
    ScaleA=true
    ScaleRGB=true
   )
   &ScaleMeshSize(Scale Factor=&VectorFromFloat(Value=&FloatFromCurve(
      FloatCurve=DataInterface() {propertyValues {Curve {keys [
          {arriveTangent=41.8211, interpMode=RCIM_Cubic, leaveTangent=41.8211, tangentMode=RCTM_User, time=0, value=0}
          {interpMode=RCIM_Cubic, tangentMode=RCTM_User, time=0.02, value=1}
          {interpMode=RCIM_Cubic, tangentMode=RCTM_User, time=0.5, value=1}
          {arriveTangent=-1.4032, interpMode=RCIM_Cubic, leaveTangent=-1.4032, tangentMode=RCTM_User, time=1, value=0.5}
         ]}}}
      Scale Curve=&FN_CameraDistance(
       0Dist:NiagaraFloat=40
       1Dist:NiagaraFloat=90
       Camera Query:NiagaraDataInterfaceCamera=DataInterface()
       Position:NiagaraPosition=$Particles.Position
      )
     )))
   &ScaleMeshSize(Scale Factor=&VectorFromFloat(Value=&Lerp_Float(
      A=1
      B=0.9
      Alpha=&Multiply_Float(
       A=&DotProduct(
        Vector A=&NormalizeVector(
         Vector To Normalize=&ReturnCameraVector(Camera Vectors=::NewEnumerator1, Coordinate Space=::World)
         Safe Normalize=true
         Safe Fallback Threshold=0
         Safe Fallback Vector=<0 0 0.001>
        )
        Vector B=&NormalizeVector(Vector To Normalize=$Engine.Owner.SystemZAxis, Safe Normalize=true, Safe Fallback Threshold=0, Safe Fallback Vector=<0 0 0.001>)
       )
       B=-1
      )
     ))) {module=ScaleMeshSize001}
  ]
  rendererClasses [&NiagaraMeshRendererProperties]
  renderers [{rendererClass=&NiagaraMeshRendererProperties, rendererIndex=0}]
 }
 Cookies {
  bEnabled=false
  emitterUpdateScript [
   &EmitterState(Life Cycle Mode=::NewEnumerator1, Loop Behavior=::NewEnumerator1, Loop Duration Mode=::NewEnumerator1, UseLoopDelay=false)
   &SpawnRate(SpawnRate=10, Use Spawn Probability=false) {enabled=false, module=SpawnRate001}
   &SpawnRate(
    Spawn Probability=&MakeCustomFloatFromBool(Bool=&SetBoolByFloatComparison(A=$User.ParaGlideLeanAlpha, B=0.1), True Float=1, False Float=0)
    SpawnRate=10
    Use Spawn Probability=true
   )
  ]
  particleSpawnScript [
   /SetVariables(Particles.VisibilityTag=&SpawnGroup()) {isSetParameters=true, module=SetVariables_#7}
   &PartitionParticles()
   &InitializeParticle(
    Lifetime Min=1.5
    Lifetime Mode=::NewEnumerator1
    Mesh Scale Max=<0.55 0.55 1.1>
    Mesh Scale Min=<0.325 0.325 0.325>
    Mesh Scale Mode=::NewEnumerator2
    UsePositionOffset=false
    Write Material Random=true
   )
   &InitialMeshOrientation(
    Mesh Orientation Mode=::NewEnumerator1
    Rotation=&MakeVector(X=0, Y=0, Z=&RandomRangeFloat(Minimum=0, Maximum=1))
    Use Rotation Vector=true
   )
  ]
  particleUpdateScript [
   /SetVariables(Particles.MeshIndex=&MakeIntFromFloat(
     Float=&FloatFromCurve(FloatCurve=DataInterface() {propertyValues {Curve {keys [<at 0 0 RCIM_Constant>,<at 0.2 1 RCIM_Constant>,<at 0.4 2 RCIM_Constant>,<at 0.6 3 RCIM_Constant>,<at 0.8 4 RCIM_Constant>]}}})
     Conversion Mode=::NewEnumerator1
    )) {isSetParameters=true, module=SetVariables_#8}
   /SetVariables(Particles.MeshIndex=&MakeIntFromFloat(
     Float=&MakeCustomFloatFromBool(
      Bool=&SetBoolByFloatComparison(A=$User.ParaGlideLeanAlpha, B=0.5)
      True Float=&MakeFloatFromInt(Int=&MakeIntFromFloat(
        Float=&FloatFromCurve(FloatCurve=DataInterface() {propertyValues {Curve {keys [<at 0.2 0 RCIM_Constant>,<at 0.4 1 RCIM_Constant>,<at 0.6 2 RCIM_Constant>,<at 0.75 3 RCIM_Constant>,<at 0.9 4 RCIM_Constant>]}}})
        Conversion Mode=::NewEnumerator1
       ))
      False Float=0
     )
     Conversion Mode=::NewEnumerator1
    )) {enabled=false, isSetParameters=true, module=SetVariables_#9}
   &ScaleColor(
    Scale RGB=&MakeVectorFromLinearColorRGB(Color=&ColorFromCurve(ColorCurve=DataInterface() {propertyValues {
        AlphaCurve {keys [<at 0 1 RCIM_Cubic>,<at 1 0 RCIM_Cubic>]}
        BlueCurve {keys [{arriveTangent=0.5432, interpMode=RCIM_Cubic, leaveTangent=0.5432, tangentMode=RCTM_User, time=0, value=0.3}, <at 1 0.9 RCIM_Cubic>]}
        GreenCurve {keys [{arriveTangent=0.5432, interpMode=RCIM_Cubic, leaveTangent=0.5432, tangentMode=RCTM_User, time=0, value=0.3}, <at 1 0.9 RCIM_Cubic>]}
        RedCurve {keys [{arriveTangent=0.5432, interpMode=RCIM_Cubic, leaveTangent=0.5432, tangentMode=RCTM_User, time=0, value=0.3}, <at 1 0.9 RCIM_Cubic>]}
       }}))
    ScaleA=true
    ScaleRGB=true
   ) {enabled=false}
   &ParticleState()
   &ScaleMeshSize(Scale Factor=&VectorFromFloat(Value=&FloatFromCurve(FloatCurve=DataInterface() {propertyValues {Curve {
         defaultValue=1
         keys [
          {arriveTangent=3.1602, interpMode=RCIM_Cubic, leaveTangent=3.1602, tangentMode=RCTM_User, time=0, value=0}
          {arriveTangent=-0.6348, interpMode=RCIM_Cubic, leaveTangent=-0.0124, tangentMode=RCTM_Break, time=0.15, value=1}
          {arriveTangent=0.0083, interpMode=RCIM_Cubic, leaveTangent=-1.0191, tangentMode=RCTM_Break, time=0.9, value=1}
          <at 1.2 0 RCIM_Cubic>
         ]
        }}})))
   &ScaleMeshSize(Scale Factor=&VectorFromFloat(Value=&Lerp_Float(
      A=1
      B=0.85
      Alpha=&Multiply_Float(
       A=&DotProduct(
        Vector A=&NormalizeVector(
         Vector To Normalize=&ReturnCameraVector(Camera Vectors=::NewEnumerator1, Coordinate Space=::World)
         Safe Normalize=true
         Safe Fallback Threshold=0
         Safe Fallback Vector=<0 0 0.001>
        )
        Vector B=&NormalizeVector(Vector To Normalize=$Engine.Owner.SystemZAxis, Safe Normalize=true, Safe Fallback Threshold=0, Safe Fallback Vector=<0 0 0.001>)
       )
       B=-1
      )
     ))) {module=ScaleMeshSize001}
   &ScaleMeshSize(
    Initial Mesh Scale:Vector3f=$Particles.Scale
    Scale Factor=&Lerp_Vector(A=<1 1 1>, B=<0 0 0>, Alpha=$System.Timeline002.NormalizedTime)
   ) {module=ScaleMeshSize002}
   &SkeletalMeshLocation(
    Bone / Socket Index=&SelectIntFromArray(Direct Array Index=$Particles.PartitionParticles.Partition, Int Selection Array=DataInterface() {propertyValues {IntData [0,1]}})
    MaskSpawnGroup=false
    Mesh Sampling Type=::NewEnumerator5
    Orientation Sampling=::NewEnumerator1
    Position Sampling=::NewEnumerator1
    Sampled Position Offset=<2 6 5>
    Sampled Position Offset Coordinate Space:ENiagaraSkelMeshTransforms=::NewEnumerator1
    Skeletal Mesh=$System.Skeletal Mesh
    Socket Sampling Mode=::NewEnumerator5
   )
   &KillParticles(Kill Particles=HlslExpression() {hlslExpression="Particles.MeshIndex == 5"})
   &GravityForce(Gravity=<0 0 125>)
   &Drag(Drag=&Lerp_Float(A=2.5, B=1, Alpha=$User.ParaGlideLeanAlpha), Use Linear Drag=true, Use Rotational Drag=false)
   &OffsetPosition(
    Jitter Amount:NiagaraFloat=&FloatFromCurve(
     FloatCurve=DataInterface() {propertyValues {Curve {keys [<at 0.2 0 RCIM_Constant>,<at 0.34 1 RCIM_Constant>,<at 0.35 0 RCIM_Constant>,<at 0.54 1 RCIM_Constant>,<at 0.55 0 RCIM_Constant>,<at 0.69 1 RCIM_Constant>,<at 0.7 0 RCIM_Constant>,<at 0.84 1 RCIM_Constant>,<at 0.85 0 RCIM_Constant>]}}}
     Scale Curve=5
    )
    Jitter Mode=1
    Jitter Offset Coordinate Space:ENiagaraCoordinateSpace=::Local
    Offset Mode=1
   ) {enabled=false}
   /SetVariables(Particles.PositionNew:NiagaraPosition=&AddVectorToPosition(
     Position=&MakeVector(
      X=&MakeCustomFloatFromBool(
       Bool=&SetBoolByFloatComparison(
        A=&FloatFromCurve(FloatCurve=DataInterface() {propertyValues {Curve {keys [<at 0 0 RCIM_Constant>,<at 0.18 1 RCIM_Constant>,<at 0.19 0 RCIM_Constant>,<at 0.38 1 RCIM_Constant>,<at 0.39 0 RCIM_Constant>,<at 0.58 1 RCIM_Constant>,<at 0.59 0 RCIM_Constant>,<at 0.78 1 RCIM_Constant>,<at 0.79 0 RCIM_Constant>,<at 0.99 1 RCIM_Constant>]}}})
        B=0
       )
       True Float=&Waveform(
        Global Input Position=&RandomRangeFloat(Minimum=0, Maximum=1)
        "Waveform [1]"=::NewEnumerator0
        "[1] Frequency"=5
        "[1] Amplitude Min/Max"=<-1 1>
       )
       False Float=0
      )
      Y=&MakeCustomFloatFromBool(
       Bool=&SetBoolByFloatComparison(
        A=&FloatFromCurve(FloatCurve=DataInterface() {propertyValues {Curve {keys [<at 0 0 RCIM_Constant>,<at 0.18 1 RCIM_Constant>,<at 0.19 0 RCIM_Constant>,<at 0.38 1 RCIM_Constant>,<at 0.39 0 RCIM_Constant>,<at 0.58 1 RCIM_Constant>,<at 0.59 0 RCIM_Constant>,<at 0.78 1 RCIM_Constant>,<at 0.79 0 RCIM_Constant>,<at 0.99 1 RCIM_Constant>]}}})
        B=0
       )
       True Float=&Waveform(
        Global Input Position=&RandomRangeFloat(Minimum=0, Maximum=1)
        "Waveform [1]"=::NewEnumerator0
        "[1] Frequency"=5
        "[1] Amplitude Min/Max"=<-1 1>
       )
       False Float=0
      )
      Z=0
     )
     Vector=Linked() {linkedVariable {name=Particles.Position, type=NiagaraPosition}}
    )) {isSetParameters=true, module=SetVariables_#10}
   &self:WriteDeath2(Position Array:NiagaraDataInterfaceArrayPosition=$System.DeathArray) {module=WriteDeath2}
   &SolveForcesAndVelocity(Clamp Velocity=false, Limit Acceleration=false)
   /SetVariables(Particles.MeshOrientation=&MultiplyQuaternion(
     Quaternion A=&LerpQuaternion(
      Quaternion A:Quat4f=<0 0 0 1>
      Quaternion B:Quat4f=&MakeQuaternion(
       Quaternion From=0
       Axis=<1 1 1>
       Angle=&RandomRangeFloat(Minimum=-33, Maximum=33)
       From=?
       To=?
       Stabilization Mode:ENiagara_OrientVectorCount=::NewEnumerator0
       Side Direction:Vector3f=?
       Up Direction:Vector3f=?
       Angle Type=::NewEnumerator0
       Normalize:NiagaraBool=true
      )
      Lerp Factor:NiagaraFloat=&FloatFromCurve(FloatCurve=DataInterface() {propertyValues {Curve {keys [{arriveTangent=5.7529, interpMode=RCIM_Cubic, leaveTangent=5.7529, tangentMode=RCTM_User, time=0, value=0}, <at 0.5 1 RCIM_Cubic>]}}})
      Enable Spherical Linear Interpolation:NiagaraBool=false
     )
     Quaternion B=&MakeQuaternion(
      Quaternion From=0
      Axis=<0 0 1>
      Angle=&Multiply_Float(A=&RandomRangeFloat(Minimum=50, Maximum=200), B=$Emitter.Age)
      From=?
      To=?
      Stabilization Mode:ENiagara_OrientVectorCount=::NewEnumerator0
      Side Direction:Vector3f=?
      Up Direction:Vector3f=?
      Angle Type=::NewEnumerator0
      Normalize:NiagaraBool=true
     )
    )) {isSetParameters=true, module=SetVariables_#11}
  ]
  rendererClasses [&NiagaraMeshRendererProperties]
  renderers [{rendererClass=&NiagaraMeshRendererProperties, rendererIndex=0}]
 }
 Cookies001 {
  bEnabled=false
  emitterUpdateScript [
   &EmitterState(Life Cycle Mode=::NewEnumerator1, Loop Behavior=::NewEnumerator1, Loop Duration Mode=::NewEnumerator1, UseLoopDelay=false)
   &UpdateValueStateChange(InputFloatParam:NiagaraFloat=$User.ParaGlideLeanAlpha, ParamType:ENiagaraValueStateChange=::NewEnumerator1) {enabled=false}
   /Niagara/Modules/Emitter/SpawnByTrigger(
    Spawn Count=10
    Trigger:NiagaraBool=&SetBoolByFloatComparison(A=$User.ParaGlideLeanAlpha, B=0.8, Comparison Type=::NewEnumerator6)
    TriggerIntervalSwitch=false
   )
   &SpawnBurst_Instantaneous(Spawn Count=10, Use Spawn Probability=false, UseLoopCountLimit=false) {enabled=false}
   &SpawnRate(SpawnRate=10, Use Spawn Probability=false) {enabled=false}
   &SpawnBurst_Instantaneous(Spawn Count=10, Spawn Time=9.5167, Use Spawn Probability=false, UseLoopCountLimit=false) {enabled=false, module=SpawnBurst_Instantaneous001}
  ]
  particleSpawnScript [
   /SetVariables(Particles.VisibilityTag=&SpawnGroup()) {isSetParameters=true, module=SetVariables_#7}
   &PartitionParticles(Particles Per Partition=5, Partition Mode:ENiagara_PartitionMode=::NewEnumerator1)
   &InitializeParticle(
    Color=&Multiply_LinearColorByFloat(
     Linear Color=&MergeHelper_ColorSwitch(
      0=&RandomRangeLinearColor(Minimum=<rgba 0.35 0.35 0.35 1>, Maximum=<rgba 0.45 0.45 0.45 1>)
      1=&RandomRangeLinearColor(Minimum=<rgba 0.5 0.5 0.5 1>, Maximum=<rgba 0.6 0.6 0.6 1>)
      2=&RandomRangeLinearColor(Minimum=<rgba 0.6 0.6 0.6 1>, Maximum=<rgba 0.7 0.7 0.7 1>)
      3=&RandomRangeLinearColor(Minimum=<rgba 0.7 0.7 0.7 1>, Maximum=<rgba 0.8 0.8 0.8 1>)
      4=&RandomRangeLinearColor(Minimum=<rgba 0.9 0.9 0.9 1>, Maximum=<rgba 0.9688 0.9688 0.9688 1>)
      Index=$Particles.PartitionParticles.Index In Partition
      Num=::NewEnumerator3
     )
     Float=&RandomRangeFloat(Minimum=0.7, Maximum=1)
    )
    Lifetime Max=1.2
    Lifetime Mode=::NewEnumerator1
    Mesh Scale Max=<0.55 0.55 1.1>
    Mesh Scale Min=<0.325 0.325 0.325>
    Mesh Scale Mode=::NewEnumerator2
    UsePositionOffset=false
    Write Material Random=true
   )
   &InitialMeshOrientation(
    Mesh Orientation Mode=::NewEnumerator1
    Rotation=&MakeVector(X=0, Y=0, Z=&RandomRangeFloat(Minimum=0, Maximum=1))
    Use Rotation Vector=true
   )
  ]
  particleUpdateScript [
   /SetVariables(Particles.MeshIndex=&MakeIntFromFloat(
     Float=&MakeCustomFloatFromBool(
      Bool=&SetBoolByFloatComparison(A=$User.ParaGlideLeanAlpha, B=0.5)
      True Float=&MakeFloatFromInt(Int=&MakeIntFromFloat(
        Float=&FloatFromCurve(FloatCurve=DataInterface() {propertyValues {Curve {keys [<at 0.2 0 RCIM_Constant>,<at 0.4 1 RCIM_Constant>,<at 0.6 2 RCIM_Constant>,<at 0.75 3 RCIM_Constant>,<at 0.9 4 RCIM_Constant>]}}})
        Conversion Mode=::NewEnumerator1
       ))
      False Float=0
     )
     Conversion Mode=::NewEnumerator1
    )) {enabled=false, isSetParameters=true, module=SetVariables_#8}
   &ParticleState(Kill Particles When Lifetime Has Elapsed=false)
   &ScaleMeshSize(Scale Factor=&VectorFromFloat(Value=&FloatFromCurve(FloatCurve=DataInterface() {propertyValues {Curve {
         defaultValue=1
         keys [
          {arriveTangent=3.1602, interpMode=RCIM_Cubic, leaveTangent=3.1602, tangentMode=RCTM_User, time=0, value=0}
          {arriveTangent=-0.6348, interpMode=RCIM_Cubic, leaveTangent=-0.0124, tangentMode=RCTM_Break, time=0.15, value=1}
          {arriveTangent=0.0083, interpMode=RCIM_Cubic, leaveTangent=-1.0191, tangentMode=RCTM_Break, time=0.9, value=1}
          <at 1.2 0 RCIM_Cubic>
         ]
        }}}))) {enabled=false}
   &ScaleMeshSize(Scale Factor=&VectorFromFloat(Value=&Lerp_Float(
      A=1.05
      B=0.9
      Alpha=&Multiply_Float(
       A=&DotProduct(
        Vector A=&NormalizeVector(
         Vector To Normalize=&ReturnCameraVector(Camera Vectors=::NewEnumerator1, Coordinate Space=::World)
         Safe Normalize=true
         Safe Fallback Threshold=0
         Safe Fallback Vector=<0 0 0.001>
        )
        Vector B=&NormalizeVector(Vector To Normalize=$Engine.Owner.SystemZAxis, Safe Normalize=true, Safe Fallback Threshold=0, Safe Fallback Vector=<0 0 0.001>)
       )
       B=-1
      )
     ))) {module=ScaleMeshSize001}
   &SkeletalMeshLocation(
    Bone / Socket Index=&SelectIntFromArray(Direct Array Index=$Particles.PartitionParticles.Partition, Int Selection Array=DataInterface() {propertyValues {IntData [0,1]}})
    MaskSpawnGroup=false
    Mesh Sampling Type=::NewEnumerator5
    Orientation Sampling=::NewEnumerator1
    Position Sampling=::NewEnumerator1
    Sampled Position Offset=<2 6 5>
    Sampled Position Offset Coordinate Space:ENiagaraSkelMeshTransforms=::NewEnumerator1
    Skeletal Mesh=$System.Skeletal Mesh
    Socket Sampling Mode=::NewEnumerator5
   )
   &KillParticles(Kill Particles=&SetBoolByFloatComparison(A=$System.Timeline.NormalizedTime, B=1, Comparison Type=::NewEnumerator4))
   &GravityForce(Gravity=&Lerp_Vector(A=<0 0 150>, B=<0 0 150>, Alpha=$User.ParaGlideLeanAlpha)) {enabled=false}
   &Drag(Drag=&Lerp_Float(A=2.5, B=0, Alpha=$User.ParaGlideLeanAlpha), Use Linear Drag=true, Use Rotational Drag=false) {enabled=false}
   &self:WriteDeath(Position Array:NiagaraDataInterfaceArrayPosition=$System.DeathArray) {enabled=false, module=WriteDeath}
   &SolveForcesAndVelocity(Clamp Velocity=false, Limit Acceleration=false)
   /SetVariables(
    Particles.MeshOrientation=&MultiplyQuaternion(
     Quaternion A=&LerpQuaternion(
      Quaternion A:Quat4f=<0 0 0 1>
      Quaternion B:Quat4f=&MakeQuaternion(
       Quaternion From=0
       Axis=<1 1 1>
       Angle=&RandomRangeFloat(Minimum=-33, Maximum=33)
       From=?
       To=?
       Stabilization Mode:ENiagara_OrientVectorCount=::NewEnumerator0
       Side Direction:Vector3f=?
       Up Direction:Vector3f=?
       Angle Type=::NewEnumerator0
       Normalize:NiagaraBool=true
      )
      Lerp Factor:NiagaraFloat=&FloatFromCurve(FloatCurve=DataInterface() {propertyValues {Curve {keys [{arriveTangent=5.7529, interpMode=RCIM_Cubic, leaveTangent=5.7529, tangentMode=RCTM_User, time=0, value=0}, <at 0.5 1 RCIM_Cubic>]}}})
      Enable Spherical Linear Interpolation:NiagaraBool=false
     )
     Quaternion B=&MakeQuaternion(
      Quaternion From=0
      Axis=<0 0 1>
      Angle=&Multiply_Float(A=&RandomRangeFloat(Minimum=50, Maximum=200), B=$Emitter.Age)
      From=?
      To=?
      Stabilization Mode:ENiagara_OrientVectorCount=::NewEnumerator0
      Side Direction:Vector3f=?
      Up Direction:Vector3f=?
      Angle Type=::NewEnumerator0
      Normalize:NiagaraBool=true
     )
    )
    Particles.Position=&Lerp_Vector(
     A=&Lerp_Vector(
      A=Linked() {linkedVariable {name=Particles.SkeletalMeshLocation.SampledPosition, type=NiagaraPosition}}
      B=&AddVectorToPosition(
       Position=$Particles.SkeletalMeshLocation.SampledPosition
       Vector=&AddVectorToPosition(
        Position=&MakeVector(
         X=0
         Y=0
         Z=&Add_Float(
          A=&MergeHelper_FloatSwitch(
           0=&MergeHelper_FloatSwitch(
            0=&Waveform(
             Global Input Position=&RandomRangeFloat(Minimum=0, Maximum=1)
             "Waveform [1]"=::NewEnumerator0
             "[1] Frequency"=1
             "[1] Amplitude Min/Max"=<-0.5 0.5>
            )
            1=&Waveform(
             Global Input Position=&RandomRangeFloat(Minimum=0, Maximum=1)
             "Waveform [1]"=::NewEnumerator0
             "[1] Frequency"=1
             "[1] Amplitude Min/Max"=<-0.5 0.5>
            )
            2=&Waveform(
             Global Input Position=&RandomRangeFloat(Minimum=0, Maximum=1)
             "Waveform [1]"=::NewEnumerator0
             "[1] Frequency"=1
             "[1] Amplitude Min/Max"=<-0.5 0.5>
            )
            3=&Waveform(
             Global Input Position=&RandomRangeFloat(Minimum=0, Maximum=1)
             "Waveform [1]"=::NewEnumerator0
             "[1] Frequency"=1
             "[1] Amplitude Min/Max"=<-0.5 0.5>
            )
            4=&Waveform(
             Global Input Position=&RandomRangeFloat(Minimum=0, Maximum=1)
             "Waveform [1]"=::NewEnumerator0
             "[1] Frequency"=1
             "[1] Amplitude Min/Max"=<-0.5 0.5>
            )
            Index=$Particles.PartitionParticles.Index In Partition
            Num=::NewEnumerator3
           )
           1=&MergeHelper_FloatSwitch(
            0=&Waveform(
             Global Input Position=&RandomRangeFloat(Minimum=0, Maximum=1)
             "Waveform [1]"=::NewEnumerator0
             "[1] Frequency"=1
             "[1] Amplitude Min/Max"=<-0.5 0.5>
            )
            1=&Waveform(
             Global Input Position=&RandomRangeFloat(Minimum=0, Maximum=1)
             "Waveform [1]"=::NewEnumerator0
             "[1] Frequency"=1
             "[1] Amplitude Min/Max"=<-0.5 0.5>
            )
            2=&Waveform(
             Global Input Position=&RandomRangeFloat(Minimum=0, Maximum=1)
             "Waveform [1]"=::NewEnumerator0
             "[1] Frequency"=1
             "[1] Amplitude Min/Max"=<-0.5 0.5>
            )
            3=&Waveform(
             Global Input Position=&RandomRangeFloat(Minimum=0, Maximum=1)
             "Waveform [1]"=::NewEnumerator0
             "[1] Frequency"=1
             "[1] Amplitude Min/Max"=<-0.5 0.5>
            )
            4=&Waveform(
             Global Input Position=&RandomRangeFloat(Minimum=0, Maximum=1)
             "Waveform [1]"=::NewEnumerator0
             "[1] Frequency"=1
             "[1] Amplitude Min/Max"=<-0.5 0.5>
            )
            Index=$Particles.PartitionParticles.Index In Partition
            Num=::NewEnumerator3
           )
           2=?
           3=?
           Index=$Particles.PartitionParticles.Partition
           Num=::NewEnumerator0
          )
          B=&MergeHelper_FloatSwitch(
           0=&Waveform(
            Global Input Position=&RandomRangeFloat(Minimum=0, Maximum=1)
            "Waveform [1]"=::NewEnumerator0
            "[1] Frequency"=1
            "[1] Amplitude Min/Max"=<-2 2>
           )
           1=&Waveform(
            Global Input Position=&RandomRangeFloat(Minimum=0, Maximum=1)
            "Waveform [1]"=::NewEnumerator0
            "[1] Frequency"=1
            "[1] Amplitude Min/Max"=<-2 2>
           )
           2=&Waveform(
            Global Input Position=&RandomRangeFloat(Minimum=0, Maximum=1)
            "Waveform [1]"=::NewEnumerator0
            "[1] Frequency"=1
            "[1] Amplitude Min/Max"=<-2 2>
           )
           3=&Waveform(
            Global Input Position=&RandomRangeFloat(Minimum=0, Maximum=1)
            "Waveform [1]"=::NewEnumerator0
            "[1] Frequency"=1
            "[1] Amplitude Min/Max"=<-2 2>
           )
           4=&Waveform(
            Global Input Position=&RandomRangeFloat(Minimum=0, Maximum=1)
            "Waveform [1]"=::NewEnumerator0
            "[1] Frequency"=1
            "[1] Amplitude Min/Max"=<-2 2>
           )
           Index=$Particles.PartitionParticles.Index In Partition
           Num=::NewEnumerator3
          )
         )
        )
        Vector=&MergeHelper_VectorSwitch(0=<0 0 7>, 1=<0 0 17>, 2=<0 0 27>, 3=<0 0 37>, 4=<0 0 47>, Index=$Particles.PartitionParticles.Index In Partition, Num=::NewEnumerator3)
       )
      )
      Alpha=&FloatFromCurve(FloatCurve=DataInterface() {propertyValues {Curve {
          defaultValue=1
          keys [<at 0.2 0 RCIM_Cubic>, {arriveTangent=1.1998, interpMode=RCIM_Cubic, leaveTangent=1.1998, tangentMode=RCTM_User, time=1, value=1}]
         }}})
     )
     B=Linked() {linkedVariable {name=Particles.SkeletalMeshLocation.SampledPosition, type=NiagaraPosition}}
     Alpha=$User.ParaGlideLeanAlpha
    )
   ) {isSetParameters=true, module=SetVariables_#11}
  ]
  rendererClasses [&NiagaraMeshRendererProperties]
  renderers [{rendererClass=&NiagaraMeshRendererProperties, rendererIndex=0}]
 }
 Cookies002 {
  bEnabled=false
  emitterUpdateScript [
   &EmitterState(Life Cycle Mode=::NewEnumerator1, Loop Behavior=::NewEnumerator1, Loop Duration Mode=::NewEnumerator1, UseLoopDelay=false)
   &SpawnRate(SpawnRate=10, Use Spawn Probability=false) {enabled=false, module=SpawnRate001}
   &SpawnRate(
    Spawn Probability=&MakeCustomFloatFromBool(Bool=&SetBoolByFloatComparison(A=$User.ParaGlideLeanAlpha, B=0.1), True Float=1, False Float=0)
    SpawnRate=10
    Use Spawn Probability=true
   )
  ]
  particleSpawnScript [
   /SetVariables(
    Particles.RadiusRandom:NiagaraFloat=&RandomRangeFloat(Minimum=0.325, Maximum=0.55)
    Particles.VisibilityTag=&SpawnGroup()
   ) {isSetParameters=true, module=SetVariables_#12}
   &PartitionParticles()
   &InitializeParticle(
    Lifetime Min=1.5
    Lifetime Mode=::NewEnumerator1
    Mesh Scale=&MakeVector(X=$Particles.RadiusRandom, Y=$Particles.RadiusRandom, Z=&RandomRangeFloat(Minimum=0.325, Maximum=1.1))
    Mesh Scale Mode=::NewEnumerator4
    UsePositionOffset=false
    Write Material Random=true
   )
   &InitialMeshOrientation(
    Mesh Orientation Mode=::NewEnumerator1
    Rotation=&MakeVector(X=0, Y=0, Z=&RandomRangeFloat(Minimum=0, Maximum=1))
    Use Rotation Vector=true
   )
  ]
  particleUpdateScript [
   /SetVariables(Particles.MeshIndex=&MakeIntFromFloat(
     Float=&FloatFromCurve(FloatCurve=DataInterface() {propertyValues {Curve {keys [<at 0 0 RCIM_Constant>,<at 0.2 1 RCIM_Constant>,<at 0.4 2 RCIM_Constant>,<at 0.6 3 RCIM_Constant>,<at 0.8 4 RCIM_Constant>]}}})
     Conversion Mode=::NewEnumerator1
    )) {isSetParameters=true, module=SetVariables_#8}
   /SetVariables(Particles.MeshIndex=&MakeIntFromFloat(
     Float=&MakeCustomFloatFromBool(
      Bool=&SetBoolByFloatComparison(A=$User.ParaGlideLeanAlpha, B=0.5)
      True Float=&MakeFloatFromInt(Int=&MakeIntFromFloat(
        Float=&FloatFromCurve(FloatCurve=DataInterface() {propertyValues {Curve {keys [<at 0.2 0 RCIM_Constant>,<at 0.4 1 RCIM_Constant>,<at 0.6 2 RCIM_Constant>,<at 0.75 3 RCIM_Constant>,<at 0.9 4 RCIM_Constant>]}}})
        Conversion Mode=::NewEnumerator1
       ))
      False Float=0
     )
     Conversion Mode=::NewEnumerator1
    )) {enabled=false, isSetParameters=true, module=SetVariables_#9}
   &ScaleColor(
    Scale RGB=&MakeVectorFromLinearColorRGB(Color=&ColorFromCurve(ColorCurve=DataInterface() {propertyValues {
        AlphaCurve {keys [<at 0 1 RCIM_Cubic>]}
        BlueCurve {keys [{arriveTangent=0.5432, interpMode=RCIM_Cubic, leaveTangent=0.5432, tangentMode=RCTM_User, time=0, value=0.3}, <at 1 0.9 RCIM_Cubic>]}
        GreenCurve {keys [{arriveTangent=0.5432, interpMode=RCIM_Cubic, leaveTangent=0.5432, tangentMode=RCTM_User, time=0, value=0.3}, <at 1 0.9 RCIM_Cubic>]}
        RedCurve {keys [{arriveTangent=0.5432, interpMode=RCIM_Cubic, leaveTangent=0.5432, tangentMode=RCTM_User, time=0, value=0.3}, <at 1 0.9 RCIM_Cubic>]}
       }}))
    ScaleA=true
    ScaleRGB=true
   )
   &ParticleState()
   &ScaleMeshSize(Scale Factor=&VectorFromFloat(Value=&FloatFromCurve(FloatCurve=DataInterface() {propertyValues {Curve {
         defaultValue=1
         keys [
          {arriveTangent=3.1602, interpMode=RCIM_Cubic, leaveTangent=3.1602, tangentMode=RCTM_User, time=0, value=0.01}
          {arriveTangent=-0.6348, interpMode=RCIM_Cubic, leaveTangent=-0.0124, tangentMode=RCTM_Break, time=0.15, value=1}
          {arriveTangent=0.0083, interpMode=RCIM_Cubic, leaveTangent=-1.0191, tangentMode=RCTM_Break, time=0.9, value=1}
          <at 1.2 0 RCIM_Cubic>
         ]
        }}})))
   &ScaleMeshSize(Scale Factor=&VectorFromFloat(Value=&Lerp_Float(
      A=1.05
      B=0.925
      Alpha=&Multiply_Float(
       A=&DotProduct(
        Vector A=&NormalizeVector(
         Vector To Normalize=&ReturnCameraVector(Camera Vectors=::NewEnumerator1, Coordinate Space=::World)
         Safe Normalize=true
         Safe Fallback Threshold=0
         Safe Fallback Vector=<0 0 0.001>
        )
        Vector B=&NormalizeVector(Vector To Normalize=$Engine.Owner.SystemZAxis, Safe Normalize=true, Safe Fallback Threshold=0, Safe Fallback Vector=<0 0 0.001>)
       )
       B=-1
      )
     ))) {module=ScaleMeshSize001}
   &ScaleMeshSize(
    Initial Mesh Scale:Vector3f=$Particles.Scale
    Scale Factor=&Lerp_Vector(A=<1 1 1>, B=<0 0 0>, Alpha=$System.Timeline002.Time)
   ) {module=ScaleMeshSize002}
   &SkeletalMeshLocation(
    Bone / Socket Index=&SelectIntFromArray(Direct Array Index=$Particles.PartitionParticles.Partition, Int Selection Array=DataInterface() {propertyValues {IntData [0,1]}})
    MaskSpawnGroup=false
    Mesh Sampling Type=::NewEnumerator5
    Orientation Sampling=::NewEnumerator1
    Position Sampling=::NewEnumerator1
    Sampled Position Offset=<2 6 15>
    Sampled Position Offset Coordinate Space:ENiagaraSkelMeshTransforms=::NewEnumerator1
    Skeletal Mesh=$System.Skeletal Mesh
    Socket Sampling Mode=::NewEnumerator5
   )
   &KillParticles(Kill Particles=&Boolean_OrOperation(Bool1=HlslExpression() {hlslExpression="Particles.MeshIndex == 5"}, Bool2=HlslExpression() {hlslExpression="Particles.Scale == 0"}))
   &GravityForce(Gravity=<0 0 125>)
   &Drag(Drag=&Lerp_Float(A=2.5, B=1, Alpha=$User.ParaGlideLeanAlpha), Use Linear Drag=true, Use Rotational Drag=false)
   &OffsetPosition(
    Jitter Amount:NiagaraFloat=&FloatFromCurve(
     FloatCurve=DataInterface() {propertyValues {Curve {keys [<at 0.2 0 RCIM_Constant>,<at 0.34 1 RCIM_Constant>,<at 0.35 0 RCIM_Constant>,<at 0.54 1 RCIM_Constant>,<at 0.55 0 RCIM_Constant>,<at 0.69 1 RCIM_Constant>,<at 0.7 0 RCIM_Constant>,<at 0.84 1 RCIM_Constant>,<at 0.85 0 RCIM_Constant>]}}}
     Scale Curve=5
    )
    Jitter Mode=1
    Jitter Offset Coordinate Space:ENiagaraCoordinateSpace=::Local
    Offset Mode=1
   ) {enabled=false}
   /SetVariables(Particles.PositionNew:NiagaraPosition=&AddVectorToPosition(
     Position=&MakeVector(
      X=&MakeCustomFloatFromBool(
       Bool=&SetBoolByFloatComparison(
        A=&FloatFromCurve(FloatCurve=DataInterface() {propertyValues {Curve {keys [<at 0 0 RCIM_Constant>,<at 0.18 1 RCIM_Constant>,<at 0.19 0 RCIM_Constant>,<at 0.38 1 RCIM_Constant>,<at 0.39 0 RCIM_Constant>,<at 0.58 1 RCIM_Constant>,<at 0.59 0 RCIM_Constant>,<at 0.78 1 RCIM_Constant>,<at 0.79 0 RCIM_Constant>,<at 0.99 1 RCIM_Constant>]}}})
        B=0
       )
       True Float=&Waveform(
        Global Input Position=&RandomRangeFloat(Minimum=0, Maximum=1)
        "Waveform [1]"=::NewEnumerator0
        "[1] Frequency"=5
        "[1] Amplitude Min/Max"=<-1 1>
       )
       False Float=0
      )
      Y=&MakeCustomFloatFromBool(
       Bool=&SetBoolByFloatComparison(
        A=&FloatFromCurve(FloatCurve=DataInterface() {propertyValues {Curve {keys [<at 0 0 RCIM_Constant>,<at 0.18 1 RCIM_Constant>,<at 0.19 0 RCIM_Constant>,<at 0.38 1 RCIM_Constant>,<at 0.39 0 RCIM_Constant>,<at 0.58 1 RCIM_Constant>,<at 0.59 0 RCIM_Constant>,<at 0.78 1 RCIM_Constant>,<at 0.79 0 RCIM_Constant>,<at 0.99 1 RCIM_Constant>]}}})
        B=0
       )
       True Float=&Waveform(
        Global Input Position=&RandomRangeFloat(Minimum=0, Maximum=1)
        "Waveform [1]"=::NewEnumerator0
        "[1] Frequency"=5
        "[1] Amplitude Min/Max"=<-1 1>
       )
       False Float=0
      )
      Z=0
     )
     Vector=Linked() {linkedVariable {name=Particles.Position, type=NiagaraPosition}}
    )) {isSetParameters=true, module=SetVariables_#10}
   &self:WriteDeath2(Position Array:NiagaraDataInterfaceArrayPosition=$System.DeathArray) {module=WriteDeath2}
   &SolveForcesAndVelocity(Clamp Velocity=false, Limit Acceleration=false)
   /SetVariables(Particles.MeshOrientation=&MultiplyQuaternion(
     Quaternion A=&LerpQuaternion(
      Quaternion A:Quat4f=<0 0 0 1>
      Quaternion B:Quat4f=&MakeQuaternion(
       Quaternion From=0
       Axis=<1 1 1>
       Angle=&RandomRangeFloat(Minimum=-33, Maximum=33)
       From=?
       To=?
       Stabilization Mode:ENiagara_OrientVectorCount=::NewEnumerator0
       Side Direction:Vector3f=?
       Up Direction:Vector3f=?
       Angle Type=::NewEnumerator0
       Normalize:NiagaraBool=true
      )
      Lerp Factor:NiagaraFloat=&FloatFromCurve(FloatCurve=DataInterface() {propertyValues {Curve {keys [{arriveTangent=5.7529, interpMode=RCIM_Cubic, leaveTangent=5.7529, tangentMode=RCTM_User, time=0, value=0}, <at 0.5 1 RCIM_Cubic>]}}})
      Enable Spherical Linear Interpolation:NiagaraBool=false
     )
     Quaternion B=&MakeQuaternion(
      Quaternion From=0
      Axis=<0 0 1>
      Angle=&Multiply_Float(A=&RandomRangeFloat(Minimum=50, Maximum=200), B=$Emitter.Age)
      From=?
      To=?
      Stabilization Mode:ENiagara_OrientVectorCount=::NewEnumerator0
      Side Direction:Vector3f=?
      Up Direction:Vector3f=?
      Angle Type=::NewEnumerator0
      Normalize:NiagaraBool=true
     )
    )) {isSetParameters=true, module=SetVariables_#11}
  ]
  rendererClasses [&NiagaraMeshRendererProperties]
  renderers [{rendererClass=&NiagaraMeshRendererProperties, rendererIndex=0}]
 }
 Cookies004 {
  bEnabled=false
  emitterUpdateScript [
   &EmitterState(Life Cycle Mode=::NewEnumerator1, Loop Behavior=::NewEnumerator1, Loop Duration Mode=::NewEnumerator1, UseLoopDelay=false)
   &SpawnRate(
    Spawn Probability=&MakeCustomFloatFromBool(Bool=&SetBoolByFloatComparison(A=$User.ParaGlideLeanAlpha, B=0.2), True Float=1, False Float=0)
    SpawnRate=10
    Use Spawn Probability=true
   )
  ]
  particleSpawnScript [
   /SetVariables(Particles.VisibilityTag=&SpawnGroup()) {isSetParameters=true, module=SetVariables_#7}
   &PartitionParticles()
   &InitializeParticle(
    Color Maximum=<rgba 0 0 0 1>
    Color Mode=::NewEnumerator2
    Lifetime Max=1.2
    Lifetime Mode=::NewEnumerator1
    Mesh Scale Mode=::NewEnumerator1
    Mesh Uniform Scale Max=0.45
    Mesh Uniform Scale Min=0.35
    UsePositionOffset=false
    Write Material Random=true
   )
   &InitialMeshOrientation(
    Mesh Orientation Mode=::NewEnumerator1
    Rotation=&MakeVector(X=0, Y=0, Z=&RandomRangeFloat(Minimum=0, Maximum=1))
    Use Rotation Vector=true
   )
  ]
  particleUpdateScript [
   /SetVariables(Particles.MeshIndex=&MakeIntFromFloat(
     Float=&FloatFromCurve(FloatCurve=DataInterface() {propertyValues {Curve {keys [<at 0.2 0 RCIM_Constant>,<at 0.4 1 RCIM_Constant>,<at 0.6 2 RCIM_Constant>,<at 0.75 3 RCIM_Constant>,<at 0.9 4 RCIM_Constant>]}}})
     Conversion Mode=::NewEnumerator1
    )) {isSetParameters=true, module=SetVariables_#8}
   /SetVariables(Particles.MeshIndex=&MakeIntFromFloat(
     Float=&MakeCustomFloatFromBool(
      Bool=&SetBoolByFloatComparison(A=$User.ParaGlideLeanAlpha, B=0.5)
      True Float=&MakeFloatFromInt(Int=&MakeIntFromFloat(
        Float=&FloatFromCurve(FloatCurve=DataInterface() {propertyValues {Curve {keys [<at 0.2 0 RCIM_Constant>,<at 0.4 1 RCIM_Constant>,<at 0.6 2 RCIM_Constant>,<at 0.75 3 RCIM_Constant>,<at 0.9 4 RCIM_Constant>]}}})
        Conversion Mode=::NewEnumerator1
       ))
      False Float=0
     )
     Conversion Mode=::NewEnumerator1
    )) {enabled=false, isSetParameters=true, module=SetVariables_#9}
   &ParticleState()
   &ScaleMeshSize(Scale Factor=&VectorFromFloat(Value=&FloatFromCurve(FloatCurve=DataInterface() {propertyValues {Curve {
         defaultValue=1
         keys [
          {arriveTangent=3.1602, interpMode=RCIM_Cubic, leaveTangent=3.1602, tangentMode=RCTM_User, time=0, value=0}
          {arriveTangent=-0.6348, interpMode=RCIM_Cubic, leaveTangent=-0.0124, tangentMode=RCTM_Break, time=0.15, value=1}
          {arriveTangent=0.0083, interpMode=RCIM_Cubic, leaveTangent=-1.0191, tangentMode=RCTM_Break, time=0.9, value=1}
          <at 1.2 0 RCIM_Cubic>
         ]
        }}})))
   &ScaleMeshSize(Scale Factor=&VectorFromFloat(Value=&Lerp_Float(
      A=1.05
      B=0.9
      Alpha=&Multiply_Float(
       A=&DotProduct(
        Vector A=&NormalizeVector(
         Vector To Normalize=&ReturnCameraVector(Camera Vectors=::NewEnumerator1, Coordinate Space=::World)
         Safe Normalize=true
         Safe Fallback Threshold=0
         Safe Fallback Vector=<0 0 0.001>
        )
        Vector B=&NormalizeVector(Vector To Normalize=$Engine.Owner.SystemZAxis, Safe Normalize=true, Safe Fallback Threshold=0, Safe Fallback Vector=<0 0 0.001>)
       )
       B=-1
      )
     ))) {module=ScaleMeshSize001}
   &SkeletalMeshLocation(
    Bone / Socket Index=&SelectIntFromArray(Direct Array Index=$Particles.PartitionParticles.Partition, Int Selection Array=DataInterface() {propertyValues {IntData [0,1]}})
    MaskSpawnGroup=false
    Mesh Sampling Type=::NewEnumerator5
    Orientation Sampling=::NewEnumerator1
    Position Sampling=::NewEnumerator1
    Sampled Position Offset=<2 6 5>
    Sampled Position Offset Coordinate Space:ENiagaraSkelMeshTransforms=::NewEnumerator1
    Skeletal Mesh=$System.Skeletal Mesh
    Socket Sampling Mode=::NewEnumerator5
   )
   &KillParticles(Kill Particles=HlslExpression() {hlslExpression="Particles.MeshIndex == 5"}) {enabled=false}
   &GravityForce(Gravity=&Lerp_Vector(A=<0 0 150>, B=<0 0 150>, Alpha=$User.ParaGlideLeanAlpha))
   &Drag(Drag=&Lerp_Float(A=2.5, B=0, Alpha=$User.ParaGlideLeanAlpha), Use Linear Drag=true, Use Rotational Drag=false)
   &self:WriteDeath(Position Array:NiagaraDataInterfaceArrayPosition=$System.DeathArray) {module=WriteDeath}
   &SolveForcesAndVelocity(Clamp Velocity=false, Limit Acceleration=false)
   /SetVariables(Particles.MeshOrientation=&MultiplyQuaternion(
     Quaternion A=&LerpQuaternion(
      Quaternion A:Quat4f=<0 0 0 1>
      Quaternion B:Quat4f=&MakeQuaternion(
       Quaternion From=0
       Axis=<1 1 1>
       Angle=&RandomRangeFloat(Minimum=-40, Maximum=40)
       From=?
       To=?
       Stabilization Mode:ENiagara_OrientVectorCount=::NewEnumerator0
       Side Direction:Vector3f=?
       Up Direction:Vector3f=?
       Angle Type=::NewEnumerator0
       Normalize:NiagaraBool=true
      )
      Lerp Factor:NiagaraFloat=&FloatFromCurve(FloatCurve=DataInterface() {propertyValues {Curve {keys [{arriveTangent=5.7529, interpMode=RCIM_Cubic, leaveTangent=5.7529, tangentMode=RCTM_User, time=0, value=0}, <at 0.5 1 RCIM_Cubic>]}}})
      Enable Spherical Linear Interpolation:NiagaraBool=false
     )
     Quaternion B=&MakeQuaternion(
      Quaternion From=0
      Axis=<0 0 1>
      Angle=&Multiply_Float(A=&RandomRangeFloat(Minimum=50, Maximum=200), B=$Emitter.Age)
      From=?
      To=?
      Stabilization Mode:ENiagara_OrientVectorCount=::NewEnumerator0
      Side Direction:Vector3f=?
      Up Direction:Vector3f=?
      Angle Type=::NewEnumerator0
      Normalize:NiagaraBool=true
     )
    )) {isSetParameters=true, module=SetVariables_#11}
  ]
  rendererClasses [&NiagaraMeshRendererProperties]
  renderers [{rendererClass=&NiagaraMeshRendererProperties, rendererIndex=0}]
 }
 Cookies_Lean {
  emitterUpdateScript [
   &EmitterState(Life Cycle Mode=::NewEnumerator1, Loop Behavior=::NewEnumerator1, Loop Duration Mode=::NewEnumerator1, UseLoopDelay=false)
   &SpawnRate(SpawnRate=10, Use Spawn Probability=false) {enabled=false, module=SpawnRate001}
   &SpawnRate(
    Spawn Probability=&MakeCustomFloatFromBool(Bool=&SetBoolByFloatComparison(A=$User.ParaGlideLeanAlpha, B=0.1), True Float=1, False Float=0)
    SpawnRate=10
    Use Spawn Probability=true
   )
  ]
  particleSpawnScript [
   /SetVariables(
    Particles.RadiusRandom:NiagaraFloat=&RandomRangeFloat(Minimum=0.325, Maximum=0.55)
    Particles.VisibilityTag=&SpawnGroup()
   ) {isSetParameters=true, module=SetVariables_#12}
   &PartitionParticles()
   &InitializeParticle(
    Lifetime Min=1.5
    Lifetime Mode=::NewEnumerator1
    Mesh Scale=&MakeVector(X=$Particles.RadiusRandom, Y=$Particles.RadiusRandom, Z=&RandomRangeFloat(Minimum=0.325, Maximum=1.1))
    Mesh Scale Mode=::NewEnumerator4
    UsePositionOffset=false
    Write Material Random=true
   )
   &InitialMeshOrientation(
    Mesh Orientation Mode=::NewEnumerator1
    Rotation=&MakeVector(X=0, Y=0, Z=&RandomRangeFloat(Minimum=0, Maximum=1))
    Use Rotation Vector=true
   )
  ]
  particleUpdateScript [
   /SetVariables(Particles.MeshIndex=&MakeIntFromFloat(
     Float=&FloatFromCurve(FloatCurve=DataInterface() {propertyValues {Curve {keys [<at 0 0 RCIM_Constant>,<at 0.2 1 RCIM_Constant>,<at 0.4 2 RCIM_Constant>,<at 0.6 3 RCIM_Constant>,<at 0.8 4 RCIM_Constant>,<at 0.95 5 RCIM_Constant>]}}})
     Conversion Mode=::NewEnumerator1
    )) {isSetParameters=true, module=SetVariables_#8}
   &UpdateValueStateChange(InputIntParam:NiagaraInt32=$Particles.MeshIndex)
   /SetVariables(Particles.MeshIndex=&MakeIntFromFloat(
     Float=&MakeCustomFloatFromBool(
      Bool=&SetBoolByFloatComparison(A=$User.ParaGlideLeanAlpha, B=0.5)
      True Float=&MakeFloatFromInt(Int=&MakeIntFromFloat(
        Float=&FloatFromCurve(FloatCurve=DataInterface() {propertyValues {Curve {keys [<at 0.2 0 RCIM_Constant>,<at 0.4 1 RCIM_Constant>,<at 0.6 2 RCIM_Constant>,<at 0.75 3 RCIM_Constant>,<at 0.9 4 RCIM_Constant>]}}})
        Conversion Mode=::NewEnumerator1
       ))
      False Float=0
     )
     Conversion Mode=::NewEnumerator1
    )) {enabled=false, isSetParameters=true, module=SetVariables_#9}
   &ScaleColor(
    Scale RGB=&MakeVectorFromLinearColorRGB(Color=&ColorFromCurve(ColorCurve=DataInterface() {propertyValues {
        AlphaCurve {keys [<at 0 1 RCIM_Cubic>]}
        BlueCurve {keys [{arriveTangent=0.5432, interpMode=RCIM_Cubic, leaveTangent=0.5432, tangentMode=RCTM_User, time=0, value=0.3}, <at 1 0.9 RCIM_Cubic>]}
        GreenCurve {keys [{arriveTangent=0.5432, interpMode=RCIM_Cubic, leaveTangent=0.5432, tangentMode=RCTM_User, time=0, value=0.3}, <at 1 0.9 RCIM_Cubic>]}
        RedCurve {keys [{arriveTangent=0.5432, interpMode=RCIM_Cubic, leaveTangent=0.5432, tangentMode=RCTM_User, time=0, value=0.3}, <at 1 0.9 RCIM_Cubic>]}
       }}))
    ScaleA=true
    ScaleRGB=true
   )
   &ParticleState()
   &ScaleMeshSize(Scale Factor=&VectorFromFloat(Value=&FloatFromCurve(FloatCurve=DataInterface() {propertyValues {Curve {
         defaultValue=1
         keys [
          {arriveTangent=3.1602, interpMode=RCIM_Cubic, leaveTangent=3.1602, tangentMode=RCTM_User, time=0, value=0.01}
          {arriveTangent=-0.6348, interpMode=RCIM_Cubic, leaveTangent=-0.0124, tangentMode=RCTM_Break, time=0.15, value=1}
          {arriveTangent=0.0083, interpMode=RCIM_Cubic, leaveTangent=-1.0191, tangentMode=RCTM_Break, time=0.9, value=1}
          <at 1.2 0 RCIM_Cubic>
         ]
        }}})))
   &ScaleMeshSize(Scale Factor=&VectorFromFloat(Value=&Lerp_Float(
      A=1.05
      B=0.925
      Alpha=&Multiply_Float(
       A=&DotProduct(
        Vector A=&NormalizeVector(
         Vector To Normalize=&ReturnCameraVector(Camera Vectors=::NewEnumerator1, Coordinate Space=::World)
         Safe Normalize=true
         Safe Fallback Threshold=0
         Safe Fallback Vector=<0 0 0.001>
        )
        Vector B=&NormalizeVector(Vector To Normalize=$Engine.Owner.SystemZAxis, Safe Normalize=true, Safe Fallback Threshold=0, Safe Fallback Vector=<0 0 0.001>)
       )
       B=-1
      )
     ))) {module=ScaleMeshSize001}
   &ScaleMeshSize(
    Initial Mesh Scale:Vector3f=$Particles.Scale
    Scale Factor=&Lerp_Vector(A=<1 1 1>, B=<0 0 0>, Alpha=$System.Timeline002.Time)
   ) {module=ScaleMeshSize002}
   &SkeletalMeshLocation(
    Bone / Socket Index=&SelectIntFromArray(Direct Array Index=$Particles.PartitionParticles.Partition, Int Selection Array=DataInterface() {propertyValues {IntData [0,1]}})
    MaskSpawnGroup=false
    Mesh Sampling Type=::NewEnumerator5
    Orientation Sampling=::NewEnumerator1
    Position Sampling=::NewEnumerator1
    Sampled Position Offset=<2 6 15>
    Sampled Position Offset Coordinate Space:ENiagaraSkelMeshTransforms=::NewEnumerator1
    Skeletal Mesh=$System.Skeletal Mesh
    Socket Sampling Mode=::NewEnumerator5
   )
   &KillParticles(Kill Particles=&Boolean_OrOperation(Bool1=HlslExpression() {hlslExpression="Particles.MeshIndex == 5"}, Bool2=HlslExpression() {hlslExpression="Particles.Scale == 0"}))
   &GravityForce(Gravity=<0 0 125>)
   &Drag(Drag=&Lerp_Float(A=2.5, B=1, Alpha=$User.ParaGlideLeanAlpha), Use Linear Drag=true, Use Rotational Drag=false)
   &OffsetPosition(
    Jitter Amount:NiagaraFloat=&FloatFromCurve(
     FloatCurve=DataInterface() {propertyValues {Curve {keys [<at 0.2 0 RCIM_Constant>,<at 0.34 1 RCIM_Constant>,<at 0.35 0 RCIM_Constant>,<at 0.54 1 RCIM_Constant>,<at 0.55 0 RCIM_Constant>,<at 0.69 1 RCIM_Constant>,<at 0.7 0 RCIM_Constant>,<at 0.84 1 RCIM_Constant>,<at 0.85 0 RCIM_Constant>]}}}
     Scale Curve=5
    )
    Jitter Mode=1
    Jitter Offset Coordinate Space:ENiagaraCoordinateSpace=::Local
    Offset Mode=1
   ) {enabled=false}
   /SetVariables(Particles.PositionNew:NiagaraPosition=&AddVectorToPosition(
     Position=&MakeVector(
      X=&MakeCustomFloatFromBool(
       Bool=&SetBoolByFloatComparison(
        A=&FloatFromCurve(FloatCurve=DataInterface() {propertyValues {Curve {keys [<at 0 0 RCIM_Constant>,<at 0.18 1 RCIM_Constant>,<at 0.19 0 RCIM_Constant>,<at 0.38 1 RCIM_Constant>,<at 0.39 0 RCIM_Constant>,<at 0.58 1 RCIM_Constant>,<at 0.59 0 RCIM_Constant>,<at 0.78 1 RCIM_Constant>,<at 0.79 0 RCIM_Constant>,<at 0.99 1 RCIM_Constant>]}}})
        B=0
       )
       True Float=&Waveform(
        Global Input Position=&RandomRangeFloat(Minimum=0, Maximum=1)
        "Waveform [1]"=::NewEnumerator0
        "[1] Frequency"=5
        "[1] Amplitude Min/Max"=<-1 1>
       )
       False Float=0
      )
      Y=&MakeCustomFloatFromBool(
       Bool=&SetBoolByFloatComparison(
        A=&FloatFromCurve(FloatCurve=DataInterface() {propertyValues {Curve {keys [<at 0 0 RCIM_Constant>,<at 0.18 1 RCIM_Constant>,<at 0.19 0 RCIM_Constant>,<at 0.38 1 RCIM_Constant>,<at 0.39 0 RCIM_Constant>,<at 0.58 1 RCIM_Constant>,<at 0.59 0 RCIM_Constant>,<at 0.78 1 RCIM_Constant>,<at 0.79 0 RCIM_Constant>,<at 0.99 1 RCIM_Constant>]}}})
        B=0
       )
       True Float=&Waveform(
        Global Input Position=&RandomRangeFloat(Minimum=0, Maximum=1)
        "Waveform [1]"=::NewEnumerator0
        "[1] Frequency"=5
        "[1] Amplitude Min/Max"=<-1 1>
       )
       False Float=0
      )
      Z=0
     )
     Vector=Linked() {linkedVariable {name=Particles.Position, type=NiagaraPosition}}
    )) {isSetParameters=true, module=SetVariables_#10}
   &self:WriteDeath2(Position Array:NiagaraDataInterfaceArrayPosition=$System.DeathArray) {module=WriteDeath2}
   &SolveForcesAndVelocity(Clamp Velocity=false, Limit Acceleration=false)
   /SetVariables(Particles.MeshOrientation=&MultiplyQuaternion(
     Quaternion A=&LerpQuaternion(
      Quaternion A:Quat4f=<0 0 0 1>
      Quaternion B:Quat4f=&MakeQuaternion(
       Quaternion From=0
       Axis=<1 1 1>
       Angle=&RandomRangeFloat(Minimum=-33, Maximum=33)
       From=?
       To=?
       Stabilization Mode:ENiagara_OrientVectorCount=::NewEnumerator0
       Side Direction:Vector3f=?
       Up Direction:Vector3f=?
       Angle Type=::NewEnumerator0
       Normalize:NiagaraBool=true
      )
      Lerp Factor:NiagaraFloat=&FloatFromCurve(FloatCurve=DataInterface() {propertyValues {Curve {keys [{arriveTangent=5.7529, interpMode=RCIM_Cubic, leaveTangent=5.7529, tangentMode=RCTM_User, time=0, value=0}, <at 0.5 1 RCIM_Cubic>]}}})
      Enable Spherical Linear Interpolation:NiagaraBool=false
     )
     Quaternion B=&MakeQuaternion(
      Quaternion From=0
      Axis=<0 0 1>
      Angle=&Multiply_Float(A=&RandomRangeFloat(Minimum=50, Maximum=200), B=$Emitter.Age)
      From=?
      To=?
      Stabilization Mode:ENiagara_OrientVectorCount=::NewEnumerator0
      Side Direction:Vector3f=?
      Up Direction:Vector3f=?
      Angle Type=::NewEnumerator0
      Normalize:NiagaraBool=true
     )
    )) {isSetParameters=true, module=SetVariables_#11}
  ]
  rendererClasses [&NiagaraMeshRendererProperties]
  renderers [{rendererClass=&NiagaraMeshRendererProperties, rendererIndex=0}]
 }
 Cookies_NoLean {
  emitterUpdateScript [
   &EmitterState(Life Cycle Mode=::NewEnumerator1, Loop Behavior=::NewEnumerator1, Loop Duration Mode=::NewEnumerator1, UseLoopDelay=false)
   &UpdateValueStateChange(InputFloatParam:NiagaraFloat=$User.ParaGlideLeanAlpha, ParamType:ENiagaraValueStateChange=::NewEnumerator1) {enabled=false}
   &SpawnByTrigger(
    Spawn Count=10
    Trigger:NiagaraBool=&SetBoolByFloatComparison(A=$System.Timeline.Time, B=0.41, Comparison Type=::NewEnumerator6)
    TriggerIntervalSwitch=false
   )
   &SpawnByTrigger(
    Spawn Count=10
    Trigger:NiagaraBool=&SetBoolByFloatComparison(A=$User.ParaGlideLeanAlpha, B=0.5, Comparison Type=::NewEnumerator6)
    TriggerIntervalSwitch=false
   ) {enabled=false, module=SpawnByTrigger001}
   &SpawnBurst_Instantaneous(Spawn Count=10, Use Spawn Probability=false, UseLoopCountLimit=false) {enabled=false}
   &SpawnRate(SpawnRate=10, Use Spawn Probability=false) {enabled=false}
   &SpawnBurst_Instantaneous(Spawn Count=10, Spawn Time=9.5167, Use Spawn Probability=false, UseLoopCountLimit=false) {enabled=false, module=SpawnBurst_Instantaneous001}
  ]
  particleSpawnScript [
   /SetVariables(
    Particles.RadiusRandom:NiagaraFloat=&RandomRangeFloat(Minimum=0.325, Maximum=0.55)
    Particles.VisibilityTag=&SpawnGroup()
   ) {isSetParameters=true, module=SetVariables_#7}
   &PartitionParticles(Particles Per Partition=5, Partition Mode:ENiagara_PartitionMode=::NewEnumerator1)
   &InitializeParticle(
    Color=&Multiply_LinearColorByFloat(
     Linear Color=&MergeHelper_ColorSwitch(
      0=&RandomRangeLinearColor(Minimum=<rgba 0.35 0.35 0.35 1>, Maximum=<rgba 0.45 0.45 0.45 1>)
      1=&RandomRangeLinearColor(Minimum=<rgba 0.5 0.5 0.5 1>, Maximum=<rgba 0.6 0.6 0.6 1>)
      2=&RandomRangeLinearColor(Minimum=<rgba 0.6 0.6 0.6 1>, Maximum=<rgba 0.7 0.7 0.7 1>)
      3=&RandomRangeLinearColor(Minimum=<rgba 0.7 0.7 0.7 1>, Maximum=<rgba 0.8 0.8 0.8 1>)
      4=&RandomRangeLinearColor(Minimum=<rgba 0.9 0.9 0.9 1>, Maximum=<rgba 0.9688 0.9688 0.9688 1>)
      Index=$Particles.PartitionParticles.Index In Partition
      Num=::NewEnumerator3
     )
     Float=&RandomRangeFloat(Minimum=0.7, Maximum=1)
    )
    Lifetime Max=1.2
    Lifetime Mode=::NewEnumerator1
    Mesh Scale=&MakeVector(X=$Particles.RadiusRandom, Y=$Particles.RadiusRandom, Z=&RandomRangeFloat(Minimum=0.325, Maximum=1.1))
    Mesh Scale Mode=::NewEnumerator4
    UsePositionOffset=false
    Write Material Random=true
   )
   &InitialMeshOrientation(
    Mesh Orientation Mode=::NewEnumerator1
    Rotation=&MakeVector(X=0, Y=0, Z=&RandomRangeFloat(Minimum=0, Maximum=1))
    Use Rotation Vector=true
   )
  ]
  particleUpdateScript [
   /SetVariables(Particles.MeshIndex=&MakeIntFromFloat(
     Float=&MakeCustomFloatFromBool(
      Bool=&SetBoolByFloatComparison(A=$User.ParaGlideLeanAlpha, B=0.5)
      True Float=&MakeFloatFromInt(Int=&MakeIntFromFloat(
        Float=&FloatFromCurve(FloatCurve=DataInterface() {propertyValues {Curve {keys [<at 0.2 0 RCIM_Constant>,<at 0.4 1 RCIM_Constant>,<at 0.6 2 RCIM_Constant>,<at 0.75 3 RCIM_Constant>,<at 0.9 4 RCIM_Constant>]}}})
        Conversion Mode=::NewEnumerator1
       ))
      False Float=0
     )
     Conversion Mode=::NewEnumerator1
    )) {enabled=false, isSetParameters=true, module=SetVariables_#8}
   &ParticleState(Kill Particles When Lifetime Has Elapsed=false)
   &ScaleMeshSize(Scale Factor=&VectorFromFloat(Value=&FloatFromCurve(FloatCurve=DataInterface() {propertyValues {Curve {
         defaultValue=1
         keys [
          {arriveTangent=3.1602, interpMode=RCIM_Cubic, leaveTangent=3.1602, tangentMode=RCTM_User, time=0, value=0}
          {arriveTangent=-0.6348, interpMode=RCIM_Cubic, leaveTangent=-0.0124, tangentMode=RCTM_Break, time=0.15, value=1}
          {arriveTangent=0.0083, interpMode=RCIM_Cubic, leaveTangent=-1.0191, tangentMode=RCTM_Break, time=0.9, value=1}
          <at 1.2 0 RCIM_Cubic>
         ]
        }}}))) {enabled=false}
   &ScaleMeshSize(Scale Factor=&VectorFromFloat(Value=&Lerp_Float(
      A=1.1
      B=0.9
      Alpha=&Multiply_Float(
       A=&DotProduct(
        Vector A=&NormalizeVector(
         Vector To Normalize=&ReturnCameraVector(Camera Vectors=::NewEnumerator1, Coordinate Space=::World)
         Safe Normalize=true
         Safe Fallback Threshold=0
         Safe Fallback Vector=<0 0 0.001>
        )
        Vector B=&NormalizeVector(Vector To Normalize=$Engine.Owner.SystemZAxis, Safe Normalize=true, Safe Fallback Threshold=0, Safe Fallback Vector=<0 0 0.001>)
       )
       B=-1
      )
     ))) {module=ScaleMeshSize001}
   &SkeletalMeshLocation(
    Bone / Socket Index=&SelectIntFromArray(Direct Array Index=$Particles.PartitionParticles.Partition, Int Selection Array=DataInterface() {propertyValues {IntData [0,1]}})
    MaskSpawnGroup=false
    Mesh Sampling Type=::NewEnumerator5
    Orientation Sampling=::NewEnumerator1
    Position Sampling=::NewEnumerator1
    Sampled Position Offset=<2 6 15>
    Sampled Position Offset Coordinate Space:ENiagaraSkelMeshTransforms=::NewEnumerator1
    Skeletal Mesh=$System.Skeletal Mesh
    Socket Sampling Mode=::NewEnumerator5
   )
   &KillParticles(Kill Particles=&SetBoolByFloatComparison(A=$System.Timeline.Time, B=0.4))
   &KillParticles(Kill Particles=&SetBoolByFloatComparison(A=$User.ParaGlideLeanAlpha, B=0.8)) {enabled=false, module=KillParticles001}
   &GravityForce(Gravity=&Lerp_Vector(A=<0 0 150>, B=<0 0 150>, Alpha=$User.ParaGlideLeanAlpha)) {enabled=false}
   &Drag(Drag=&Lerp_Float(A=2.5, B=0, Alpha=$User.ParaGlideLeanAlpha), Use Linear Drag=true, Use Rotational Drag=false) {enabled=false}
   &self:WriteDeath(Position Array:NiagaraDataInterfaceArrayPosition=$System.DeathArray) {enabled=false, module=WriteDeath}
   &SolveForcesAndVelocity(Clamp Velocity=false, Limit Acceleration=false)
   /SetVariables(
    Particles.MeshOrientation=&MultiplyQuaternion(
     Quaternion A=&LerpQuaternion(
      Quaternion A:Quat4f=<0 0 0 1>
      Quaternion B:Quat4f=&MakeQuaternion(
       Quaternion From=0
       Axis=<1 1 1>
       Angle=&RandomRangeFloat(Minimum=-33, Maximum=33)
       From=?
       To=?
       Stabilization Mode:ENiagara_OrientVectorCount=::NewEnumerator0
       Side Direction:Vector3f=?
       Up Direction:Vector3f=?
       Angle Type=::NewEnumerator0
       Normalize:NiagaraBool=true
      )
      Lerp Factor:NiagaraFloat=&FloatFromCurve(FloatCurve=DataInterface() {propertyValues {Curve {keys [{arriveTangent=5.7529, interpMode=RCIM_Cubic, leaveTangent=5.7529, tangentMode=RCTM_User, time=0, value=0}, <at 0.5 1 RCIM_Cubic>]}}})
      Enable Spherical Linear Interpolation:NiagaraBool=false
     )
     Quaternion B=&MakeQuaternion(
      Quaternion From=0
      Axis=<0 0 1>
      Angle=&Multiply_Float(A=&RandomRangeFloat(Minimum=50, Maximum=200), B=$Emitter.Age)
      From=?
      To=?
      Stabilization Mode:ENiagara_OrientVectorCount=::NewEnumerator0
      Side Direction:Vector3f=?
      Up Direction:Vector3f=?
      Angle Type=::NewEnumerator0
      Normalize:NiagaraBool=true
     )
    )
    Particles.Position=&Lerp_Vector(
     A=&Lerp_Vector(
      A=Linked() {linkedVariable {name=Particles.SkeletalMeshLocation.SampledPosition, type=NiagaraPosition}}
      B=&AddVectorToPosition(
       Position=$Particles.SkeletalMeshLocation.SampledPosition
       Vector=&AddVectorToPosition(
        Position=&MakeVector(
         X=0
         Y=0
         Z=&Add_Float(
          A=&MergeHelper_FloatSwitch(
           0=&MergeHelper_FloatSwitch(
            0=&Waveform(
             Global Input Position=&RandomRangeFloat(Minimum=0, Maximum=1)
             "Waveform [1]"=::NewEnumerator0
             "[1] Frequency"=1
             "[1] Amplitude Min/Max"=<-0.5 0.5>
            )
            1=&Waveform(
             Global Input Position=&RandomRangeFloat(Minimum=0, Maximum=1)
             "Waveform [1]"=::NewEnumerator0
             "[1] Frequency"=1
             "[1] Amplitude Min/Max"=<-0.5 0.5>
            )
            2=&Waveform(
             Global Input Position=&RandomRangeFloat(Minimum=0, Maximum=1)
             "Waveform [1]"=::NewEnumerator0
             "[1] Frequency"=1
             "[1] Amplitude Min/Max"=<-0.5 0.5>
            )
            3=&Waveform(
             Global Input Position=&RandomRangeFloat(Minimum=0, Maximum=1)
             "Waveform [1]"=::NewEnumerator0
             "[1] Frequency"=1
             "[1] Amplitude Min/Max"=<-0.5 0.5>
            )
            4=&Waveform(
             Global Input Position=&RandomRangeFloat(Minimum=0, Maximum=1)
             "Waveform [1]"=::NewEnumerator0
             "[1] Frequency"=1
             "[1] Amplitude Min/Max"=<-0.5 0.5>
            )
            Index=$Particles.PartitionParticles.Index In Partition
            Num=::NewEnumerator3
           )
           1=&MergeHelper_FloatSwitch(
            0=&Waveform(
             Global Input Position=&RandomRangeFloat(Minimum=0, Maximum=1)
             "Waveform [1]"=::NewEnumerator0
             "[1] Frequency"=1
             "[1] Amplitude Min/Max"=<-0.5 0.5>
            )
            1=&Waveform(
             Global Input Position=&RandomRangeFloat(Minimum=0, Maximum=1)
             "Waveform [1]"=::NewEnumerator0
             "[1] Frequency"=1
             "[1] Amplitude Min/Max"=<-0.5 0.5>
            )
            2=&Waveform(
             Global Input Position=&RandomRangeFloat(Minimum=0, Maximum=1)
             "Waveform [1]"=::NewEnumerator0
             "[1] Frequency"=1
             "[1] Amplitude Min/Max"=<-0.5 0.5>
            )
            3=&Waveform(
             Global Input Position=&RandomRangeFloat(Minimum=0, Maximum=1)
             "Waveform [1]"=::NewEnumerator0
             "[1] Frequency"=1
             "[1] Amplitude Min/Max"=<-0.5 0.5>
            )
            4=&Waveform(
             Global Input Position=&RandomRangeFloat(Minimum=0, Maximum=1)
             "Waveform [1]"=::NewEnumerator0
             "[1] Frequency"=1
             "[1] Amplitude Min/Max"=<-0.5 0.5>
            )
            Index=$Particles.PartitionParticles.Index In Partition
            Num=::NewEnumerator3
           )
           2=?
           3=?
           Index=$Particles.PartitionParticles.Partition
           Num=::NewEnumerator0
          )
          B=&MergeHelper_FloatSwitch(
           0=&Waveform(
            Global Input Position=&RandomRangeFloat(Minimum=0, Maximum=1)
            "Waveform [1]"=::NewEnumerator0
            "[1] Frequency"=1
            "[1] Amplitude Min/Max"=<-2 2>
           )
           1=&Waveform(
            Global Input Position=&RandomRangeFloat(Minimum=0, Maximum=1)
            "Waveform [1]"=::NewEnumerator0
            "[1] Frequency"=1
            "[1] Amplitude Min/Max"=<-2 2>
           )
           2=&Waveform(
            Global Input Position=&RandomRangeFloat(Minimum=0, Maximum=1)
            "Waveform [1]"=::NewEnumerator0
            "[1] Frequency"=1
            "[1] Amplitude Min/Max"=<-2 2>
           )
           3=&Waveform(
            Global Input Position=&RandomRangeFloat(Minimum=0, Maximum=1)
            "Waveform [1]"=::NewEnumerator0
            "[1] Frequency"=1
            "[1] Amplitude Min/Max"=<-2 2>
           )
           4=&Waveform(
            Global Input Position=&RandomRangeFloat(Minimum=0, Maximum=1)
            "Waveform [1]"=::NewEnumerator0
            "[1] Frequency"=1
            "[1] Amplitude Min/Max"=<-2 2>
           )
           Index=$Particles.PartitionParticles.Index In Partition
           Num=::NewEnumerator3
          )
         )
        )
        Vector=&MergeHelper_VectorSwitch(0=<0 0 7>, 1=<0 0 20>, 2=<0 0 32>, 3=<0 0 45>, 4=<0 0 58>, Index=$Particles.PartitionParticles.Index In Partition, Num=::NewEnumerator3)
       )
      )
      Alpha=&FloatFromCurve(FloatCurve=DataInterface() {propertyValues {Curve {
          defaultValue=1
          keys [<at 0.2 0 RCIM_Cubic>, {arriveTangent=1.1998, interpMode=RCIM_Cubic, leaveTangent=1.1998, tangentMode=RCTM_User, time=1, value=1}]
         }}})
     )
     B=Linked() {linkedVariable {name=Particles.SkeletalMeshLocation.SampledPosition, type=NiagaraPosition}}
     Alpha=$System.Timeline.NormalizedTime
    )
   ) {isSetParameters=true, module=SetVariables_#11}
  ]
  rendererClasses [&NiagaraMeshRendererProperties]
  renderers [{rendererClass=&NiagaraMeshRendererProperties, rendererIndex=0}]
 }
 Default_Emote_TraversalDust {
  bEnabled=false
  emitterSpawnScript [/SetVariables(Emitter.FE Particle Spawn Rate=60, Emitter.Idle Spawn Rate=60, Emitter.Max Traversal Speed=300, Emitter.Particle Spawn Rate=60) {isSetParameters=true}]
  emitterUpdateScript [
   &EmitterState()
   &SpawnRate(
    SpawnRate=&FN_FrontEndFloatModifier(
     Gameplay=&MakeCustomFloatFromBool(
      Bool=&SetBoolByFloatComparison(A=&VectorLength(Vector=$User.OwnerLocalVelocity), B=150)
      True Float=$Emitter.Particle Spawn Rate
      False Float=$Emitter.Idle Spawn Rate
     )
     Front End=$Emitter.FE Particle Spawn Rate
    )
    Use Spawn Probability=false
   )
  ]
  particleSpawnScript [
   &PartitionParticles()
   &InitializeParticle(
    Color Channel Mode:ENiagaraChannelCorrelation=::NewEnumerator0
    Color Maximum=<rgba 0.781 0.7271 0.5498 0.4>
    Color Minimum=<rgba 0.478 0.3791 0.2199 0.2>
    Color Mode=::NewEnumerator2
    Lifetime Max=0.5
    Lifetime Min=0.3
    Lifetime Mode=::NewEnumerator1
    Sprite Rotation Mode=::NewEnumerator2
    Sprite Size Mode=::NewEnumerator1
    Uniform Sprite Size Max=50
    Uniform Sprite Size Min=30
    UsePositionOffset=false
    Write Material Random=true
   ) {module=InitializeParticle001}
   &SocketLocation(
    SamplePosOffset=true
    Sampled Position Offset=<0 -10 -17>
    Sampled Position Offset Coordinate Space:ENiagaraSkelMeshTransforms=::NewEnumerator2
    Socket Index=$Particles.PartitionParticles.Partition
    Socket Sampling Mode=::NewEnumerator3
    VelocitySpeedLimit=false
   )
   &AddVelocity(Velocity=<0 2 0.2>, Velocity Speed Scale=&RandomRangeFloat(Minimum=300, Maximum=500)) {module=AddVelocity001}
   &CameraOffset(Camera Offset Amount=15)
  ]
  particleUpdateScript [
   &ParticleState()
   &Drag(
    Drag=&FloatFromCurve(
     FloatCurve=DataInterface() {propertyValues {Curve {
        defaultValue=1
        keys [{interpMode=RCIM_Cubic, tangentMode=RCTM_User, time=0, value=0.75}, {interpMode=RCIM_Cubic, tangentMode=RCTM_User, time=0.65, value=1}]
       }}}
     Scale Curve=&Lerp_Float(
      A=20
      B=8
      Alpha=&RemapRange(
       Input Value To Remap=&VectorLength(Vector=$User.OwnerLocalVelocity)
       Input Min=0
       Input Max=$Emitter.Max Traversal Speed
       Output Min=0
       Output Max=1
      )
     )
    )
    Use Linear Drag=true
    Use Rotational Drag=true
   )
   &ScaleSpriteSize(Uniform Curve Sprite Scale=DataInterface() {propertyValues {Curve {keys [
        {arriveTangent=0.8753, interpMode=RCIM_Cubic, leaveTangent=0.8753, tangentMode=RCTM_User, time=0, value=1}
        {arriveTangent=0.0238, interpMode=RCIM_Cubic, leaveTangent=0.0238, tangentMode=RCTM_User, time=1, value=1.3}
       ]}}})
   &DynamicMaterialParameters(
    Index 0 Param 1=&Add_Float(
     A=&RandomRangeFloat(Minimum=0, Maximum=1)
     B=&FloatFromCurve(
      FloatCurve=DataInterface() {propertyValues {Curve {keys [
          <at 0 0 RCIM_Cubic>
          {arriveTangent=1.5947, interpMode=RCIM_Cubic, leaveTangent=1.5947, tangentMode=RCTM_User, time=0.2557, value=0.5257}
          <at 1 1 RCIM_Cubic>
         ]}}}
      CurveIndex=$Particles.Age
      Scale Curve=0.25
     )
    )
    Module.Index0Float0Write=true
    Module.Index0Float1Write=true
    Module.Index0Float2Write=true
    Module.Index0Float3Write=true
   ) {module=DynamicMaterialParameters001}
   &ScaleColor(
    Scale Alpha=&FloatFromCurve(
     FloatCurve=DataInterface() {propertyValues {Curve {keys [
         {interpMode=RCIM_Cubic, tangentMode=RCTM_User, time=0, value=0}
         {arriveTangentWeight=0.0754, interpMode=RCIM_Cubic, leaveTangentWeight=0.2297, tangentMode=RCTM_User, time=0.1684, value=1}
         {arriveTangent=-0.5014, interpMode=RCIM_Cubic, leaveTangent=-0.5014, tangentMode=RCTM_User, time=0.4045, value=1}
         {arriveTangent=-2.9749, interpMode=RCIM_Cubic, leaveTangent=-2.9749, tangentMode=RCTM_User, time=0.9, value=0}
        ]}}}
     Scale Curve=&FN_FrontEndFloatModifier(Gameplay=0.4, Front End=0.4)
    )
    Scale RGB=&MakeVectorFromLinearColorRGB(Color=&FNTimeOfDayColor(
      Night Color=&RandomRangeLinearColor(Minimum=<rgba 0.4138 0.5201 0.6661 1>, Maximum=<rgba 0.781 0.781 0.781 1>)
      Morning Color=&RandomRangeLinearColor(Minimum=<rgba 0.6882 0.6227 0.9531 1>, Maximum=<rgba 0.6063 0.7457 0.9346 1>)
      Day Color=&RandomRangeLinearColor(Minimum=<rgba 0.478 0.4265 0.4121 0.8>, Maximum=<rgba 0.781 0.718 0.6943 0.9>)
      Evening Color=&RandomRangeLinearColor(Minimum=<rgba 0.478 0.478 0.478 1>, Maximum=<rgba 0.781 0.781 0.781 1>)
     ))
    ScaleA=true
    ScaleRGB=true
   )
   &SolveForcesAndVelocity(Clamp Velocity=false, Limit Acceleration=false)
  ]
  rendererClasses [&NiagaraSpriteRendererProperties]
  renderers [
   {rendererClass=&NiagaraSpriteRendererProperties}
   {rendererClass=&NiagaraSpriteRendererProperties, rendererIndex=1}
  ]
 }
 Gizmo_LocalSpace {
  bEnabled=false
  emitterUpdateScript [
   &EmitterState()
   &SpawnBurst_Instantaneous(Spawn Count=4, Spawn Probability=HlslExpression() {hlslExpression="Emitter.LoopCount == 0"}, Use Spawn Probability=true)
  ]
  particleSpawnScript [
   &InitializeParticle(
    Color=<rgba 0 0.3677 1 1>
    Lifetime Max=1
    Lifetime Mode=::NewEnumerator1
    Mesh Scale Mode=::NewEnumerator3
    Mesh Uniform Scale=3
    UsePositionOffset=false
    Write Material Random=true
   )
   &InitialMeshOrientation(Mesh Axis To Orient=<0 0 -1>, Mesh Orientation Mode=::NewEnumerator2, Use Rotation Vector=false)
  ]
  particleUpdateScript [
   &ParticleState(Kill Particles When Lifetime Has Elapsed=false, Let Infinitely Lived Particles Die When Emitter Deactivates=true)
   &UpdateMeshOrientation(
    Facing Direction=$System.FN_Contrail_Setup.Direction Vector
    Facing Interpolation=false
    Orientation Method=::NewEnumerator1
    Override Quat=false
    Stabilization Mode:ENiagara_OrientVectorCount=::NewEnumerator2
    TransformFacing=false
    TransformSide=false
   )
   /SetVariables(Particles.Position=&AddVectorToPosition(
     Position=$Particles.Initial.Position
     Vector=&Multiply_VectorByFloat(
      Vector=$System.FN_Contrail_Setup.Direction Velocity
      Float=&FloatFromCurve(
       FloatCurve=DataInterface() {propertyValues {Curve {keys [{tangentMode=RCTM_User, time=0, value=0}, {tangentMode=RCTM_User, time=1, value=1}]}}}
       Scale Curve=$Particles.Initial.Lifetime
      )
     )
    )) {enabled=false, isSetParameters=true, module=SetVariables_#13}
   /SetVariables(Particles.Velocity=$System.FN_Contrail_Setup.Direction Vector) {isSetParameters=true, module=SetVariables_#14}
  ]
  rendererClasses [&NiagaraMeshRendererProperties]
  renderers [{rendererClass=&NiagaraMeshRendererProperties, rendererIndex=0}]
 }
 Lines {
  emitterUpdateScript [
   &EmitterState(Life Cycle Mode=::NewEnumerator1, Loop Behavior=::NewEnumerator1, Loop Duration Mode=::NewEnumerator1, UseLoopDelay=false)
   &SpawnRate(
    Spawn Probability=0.5
    SpawnRate=&Lerp_Float(A=20, B=32, Alpha=$User.ParaGlideLeanAlpha)
    Use Spawn Probability=false
   ) {module=SpawnRate002}
  ]
  particleSpawnScript [
   /SetVariables(Particles.VisibilityTag=&SelectIntFromArray(
     Direct Array Index=&SelectIntFromWeightedDistributionArray(
      Array Sampling Mode:ENiagara_LifetimeMode=::NewEnumerator1
      Direct Array Index:NiagaraInt32=?
      Direct Array Mode:ENiagara_WrapClamp=::NewEnumerator0
      Evaluation Type:ENiagaraRandomnessEvaluation=::NewEnumerator0
      Recalculate Random Each Loop:NiagaraBool=false
      Weighted Distribution Int Array:NiagaraDataInterfaceArrayDistributionInt=DataInterface() {propertyValues {ArrayData [{value=0, weight=0.3}, {value=1, weight=0.7}]}}
      Randomness Mode:ENiagaraRandomnessMode=::NewEnumerator0
      Random Seed:NiagaraInt32=0
      Fixed Random Seed:NiagaraBool=false
     )
     Int Selection Array=DataInterface() {propertyValues {IntData [0,1]}}
    )) {isSetParameters=true, module=SetVariables_#15}
   &PartitionParticles(Partition Mode:ENiagara_PartitionMode=::NewEnumerator5)
   &InitializeParticle(Color Mode=::NewEnumerator0, Lifetime=0.2, UsePositionOffset=false, Write Material Random=true)
   &ShapeLocation(
    Cylinder Height=20
    Cylinder Radius=8
    MaskSpawnGroup=false
    Offset Mode=::NewEnumerator0
    Override Seed=false
    Shape Primitive=::NewEnumerator1
    Surface Only=false
    UseOwnerScale=true
   )
   &AddVelocity(Velocity=&MakeVector(X=0, Y=0, Z=&Lerp_Float(A=15, B=250, Alpha=$System.Timeline001.NormalizedTime))) {module=AddVelocity001}
  ]
  particleUpdateScript [
   &ParticleState()
   &ScaleSpriteSize()
   &SkeletalMeshLocation(
    Bone / Socket Index=&SelectIntFromArray(Direct Array Index=$Particles.PartitionParticles.Partition, Int Selection Array=DataInterface() {propertyValues {IntData [0,1]}})
    MaskSpawnGroup=false
    Mesh Sampling Type=::NewEnumerator5
    Orientation Sampling=::NewEnumerator1
    Position Sampling=::NewEnumerator1
    Sampled Position Offset=&AddVectorToPosition(Position=$Particles.Initial.ShapeLocation.ShapePosition, Vector=<0 0 10>)
    Sampled Position Offset Coordinate Space:ENiagaraSkelMeshTransforms=::NewEnumerator1
    Skeletal Mesh=$System.Skeletal Mesh
    Socket Sampling Mode=::NewEnumerator5
   )
   &Drag(Drag=&RandomRangeFloat(Minimum=0.8, Maximum=1.2), Use Linear Drag=true, Use Rotational Drag=true)
   &SolveForcesAndVelocity(Clamp Velocity=false, Limit Acceleration=false)
   /SetVariables(
    Particles.MotionLineColor:LinearColor=&MakeLinearColorFromVectorAndFloat(
     "Vector (RGB)"=&MakeVectorFromLinearColorRGB(Color=&RandomRangeLinearColor(Minimum=<rgba 0.08 0.0267 0 1>, Maximum=<rgba 0.3531 0.1582 0.0607 1>))
     "Float (Alpha)"=&Multiply_Float(
      A=&MergeHelper_FloatSwitch(
       0=&RandomRangeFloat(Minimum=1, Maximum=2)
       1=&RandomRangeFloat(Minimum=7, Maximum=14)
       2=?
       3=?
       Index=$Particles.VisibilityTag
       Num=::NewEnumerator0
      )
      B=&MakeFloatFromVector(
       Vector=&VectorFromFloat(Value=&Lerp_Float(
         A=1
         B=1.5
         Alpha=&Multiply_Float(
          A=&DotProduct(
           Vector A=&NormalizeVector(
            Vector To Normalize=&ReturnCameraVector(Camera Vectors=::NewEnumerator1, Coordinate Space=::World)
            Safe Normalize=true
            Safe Fallback Threshold=0
            Safe Fallback Vector=<0 0 0.001>
           )
           Vector B=&NormalizeVector(Vector To Normalize=$Engine.Owner.SystemZAxis, Safe Normalize=true, Safe Fallback Threshold=0, Safe Fallback Vector=<0 0 0.001>)
          )
          B=-1
         )
        ))
       Channel=::NewEnumerator0
      )
     )
    )
    Particles.MotionLineSize:Vector2f=&Vector2DFromCurve(
     Vector2Curve=DataInterface() {propertyValues {
       XCurve {keys [{tangentMode=RCTM_User, time=0, value=1}, {tangentMode=RCTM_User, time=1, value=1.25}]}
       YCurve {keys [{tangentMode=RCTM_User, time=0, value=0}, {tangentMode=RCTM_User, time=1, value=1}]}
      }}
     ScaleCurve=&Lerp_Vector2(
      A=<25 100>
      B=&MakeVector2D(
       X=25
       Y=&FloatFromCurve(
        FloatCurve=DataInterface() {propertyValues {Curve {keys [<at 0 0>,<at 1 1>]}}}
        CurveIndex=$System.Timeline001.NormalizedTime
        Scale Curve=110
       )
      )
      Alpha=$User.ParaGlideLeanAlpha
     )
    )
   ) {isSetParameters=true, module=SetVariables_#16}
  ]
  rendererClasses [&NiagaraSpriteRendererProperties]
  renderers [
   {rendererClass=&NiagaraSpriteRendererProperties}
   {rendererClass=&NiagaraSpriteRendererProperties, rendererIndex=1}
  ]
 }
 Lines001 {
  bEnabled=false
  emitterUpdateScript [
   &EmitterState(Life Cycle Mode=::NewEnumerator1, Loop Behavior=::NewEnumerator1, Loop Duration Mode=::NewEnumerator1, UseLoopDelay=false)
   &SpawnRate(
    Spawn Probability=0.5
    SpawnRate=&Lerp_Float(A=20, B=32, Alpha=$User.ParaGlideLeanAlpha)
    Use Spawn Probability=false
   ) {module=SpawnRate002}
  ]
  particleSpawnScript [
   /SetVariables(Particles.VisibilityTag=&SelectIntFromArray(
     Direct Array Index=&SelectIntFromWeightedDistributionArray(
      Array Sampling Mode:ENiagara_LifetimeMode=::NewEnumerator1
      Direct Array Index:NiagaraInt32=?
      Direct Array Mode:ENiagara_WrapClamp=::NewEnumerator0
      Evaluation Type:ENiagaraRandomnessEvaluation=::NewEnumerator0
      Recalculate Random Each Loop:NiagaraBool=false
      Weighted Distribution Int Array:NiagaraDataInterfaceArrayDistributionInt=DataInterface() {propertyValues {ArrayData [{value=0, weight=0.3}, {value=1, weight=0.7}]}}
      Randomness Mode:ENiagaraRandomnessMode=::NewEnumerator0
      Random Seed:NiagaraInt32=0
      Fixed Random Seed:NiagaraBool=false
     )
     Int Selection Array=DataInterface() {propertyValues {IntData [0,1]}}
    )) {isSetParameters=true, module=SetVariables_#15}
   &PartitionParticles(Partition Mode:ENiagara_PartitionMode=::NewEnumerator5)
   &InitializeParticle(Color Mode=::NewEnumerator0, Lifetime=0.2, UsePositionOffset=false, Write Material Random=true)
   &ShapeLocation(
    Cylinder Height=20
    Cylinder Radius=8
    MaskSpawnGroup=false
    Offset Mode=::NewEnumerator0
    Override Seed=false
    Shape Primitive=::NewEnumerator1
    Surface Only=false
    UseOwnerScale=true
   )
   &AddVelocity(Velocity=&MakeVector(X=0, Y=0, Z=&Lerp_Float(A=15, B=250, Alpha=$System.Timeline001.NormalizedTime))) {module=AddVelocity001}
  ]
  particleUpdateScript [
   &ParticleState()
   &ScaleSpriteSize()
   &SkeletalMeshLocation(
    Bone / Socket Index=&SelectIntFromArray(Direct Array Index=$Particles.PartitionParticles.Partition, Int Selection Array=DataInterface() {propertyValues {IntData [0,1]}})
    MaskSpawnGroup=false
    Mesh Sampling Type=::NewEnumerator5
    Orientation Sampling=::NewEnumerator1
    Position Sampling=::NewEnumerator1
    Sampled Position Offset=&AddVectorToPosition(
     Position=$Particles.Initial.ShapeLocation.ShapePosition
     Vector=&Lerp_Vector(
      A=&MaskVector(Mask Source=1, Mask Parameter=$Particles.PartitionParticles.Partition, Mask=0, Masked Vector=<0 0 -5>, Passthrough Vector=<0 0 -5>)
      B=&MaskVector(Mask Source=1, Mask Parameter=$Particles.PartitionParticles.Partition, Mask=0, Masked Vector=<0 0 -5>, Passthrough Vector=<0 0 -5>)
      Alpha=$User.ParaGlideLeanAlpha
     )
    )
    Sampled Position Offset Coordinate Space:ENiagaraSkelMeshTransforms=::NewEnumerator1
    Skeletal Mesh=$System.Skeletal Mesh
    Socket Sampling Mode=::NewEnumerator5
   )
   &Drag(Drag=&RandomRangeFloat(Minimum=0.8, Maximum=1.2), Use Linear Drag=true, Use Rotational Drag=true)
   &SolveForcesAndVelocity(Clamp Velocity=false, Limit Acceleration=false)
   /SetVariables(
    Particles.MotionLineColor:LinearColor=&MakeLinearColorFromVectorAndFloat(
     "Vector (RGB)"=&MakeVectorFromLinearColorRGB(Color=&RandomRangeLinearColor(Minimum=<rgba 0.08 0.0267 0 1>, Maximum=<rgba 0.3531 0.1582 0.0607 1>))
     "Float (Alpha)"=&Multiply_Float(
      A=&MergeHelper_FloatSwitch(
       0=&RandomRangeFloat(Minimum=1, Maximum=2)
       1=&RandomRangeFloat(Minimum=7, Maximum=14)
       2=?
       3=?
       Index=$Particles.VisibilityTag
       Num=::NewEnumerator0
      )
      B=&MakeFloatFromVector(
       Vector=&VectorFromFloat(Value=&Lerp_Float(
         A=1
         B=1.5
         Alpha=&Multiply_Float(
          A=&DotProduct(
           Vector A=&NormalizeVector(
            Vector To Normalize=&ReturnCameraVector(Camera Vectors=::NewEnumerator1, Coordinate Space=::World)
            Safe Normalize=true
            Safe Fallback Threshold=0
            Safe Fallback Vector=<0 0 0.001>
           )
           Vector B=&NormalizeVector(Vector To Normalize=$Engine.Owner.SystemZAxis, Safe Normalize=true, Safe Fallback Threshold=0, Safe Fallback Vector=<0 0 0.001>)
          )
          B=-1
         )
        ))
       Channel=::NewEnumerator0
      )
     )
    )
    Particles.MotionLineSize:Vector2f=&Vector2DFromCurve(
     Vector2Curve=DataInterface() {propertyValues {
       XCurve {keys [{tangentMode=RCTM_User, time=0, value=1}, {tangentMode=RCTM_User, time=1, value=1.25}]}
       YCurve {keys [{tangentMode=RCTM_User, time=0, value=0}, {tangentMode=RCTM_User, time=1, value=1}]}
      }}
     ScaleCurve=&Lerp_Vector2(
      A=<25 100>
      B=&MakeVector2D(
       X=25
       Y=&FloatFromCurve(
        FloatCurve=DataInterface() {propertyValues {Curve {keys [<at 0 0>,<at 1 1>]}}}
        CurveIndex=$System.Timeline001.NormalizedTime
        Scale Curve=110
       )
      )
      Alpha=$User.ParaGlideLeanAlpha
     )
    )
   ) {isSetParameters=true, module=SetVariables_#16}
  ]
  rendererClasses [&NiagaraSpriteRendererProperties]
  renderers [
   {rendererClass=&NiagaraSpriteRendererProperties}
   {rendererClass=&NiagaraSpriteRendererProperties, rendererIndex=1}
  ]
 }
 Particle_Root {
  bEnabled=false
  emitterUpdateScript [
   &EmitterState()
   &SpawnRate(SpawnRate=31, Use Spawn Probability=false)
  ]
  particleSpawnScript [
   &InitializeParticle(
    Color=<rgba 0 0.3677 1 1>
    Lifetime Max=1
    Lifetime Mode=::NewEnumerator1
    Sprite Size Mode=::NewEnumerator1
    UsePositionOffset=false
    Write Material Random=true
   )
   /SetVariables(Particles.Velocity=$System.FN_Contrail_Setup.Direction Vector) {enabled=false, isSetParameters=true}
   &ShapeLocation(
    Cylinder Height Midpoint=1
    Cylinder Radius=400
    MaskSpawnGroup=false
    Offset Mode=::NewEnumerator0
    Override Seed=false
    Rotation Axis=$System.FN_Contrail_Setup.Direction Vector
    Rotation Mode=::NewEnumerator1
    Shape Primitive=::NewEnumerator1
    Surface Only=true
    Surface Only Band Thickness=0.75
    Transform Order:ENiagara_TransformOrder=::NewEnumerator1
    Use Endcaps In Surface Only Mode=false
    UseOwnerScale=true
   )
  ]
  particleUpdateScript [
   &ParticleState()
   /SetVariables(Particles.NewAge:NiagaraFloat=&Multiply_Float(A=$Particles.Age, B=1)) {isSetParameters=true}
   /SetVariables(Particles.Velocity=&Multiply_VectorByFloat(
     Vector=&Lerp_Vector(A=$System.FN_Contrail_Setup.Direction Vector, B=$System.FN_Contrail_Setup.Direction Velocity, Alpha=$Particles.NewAge)
     Float=1
    )) {isSetParameters=true}
   /SetVariables(Particles.Position=&AddVectorToPosition(
     Position=$Particles.Initial.Position
     Vector=&Multiply_VectorByFloat(Vector=$Particles.Velocity, Float=$Particles.Age)
    )) {isSetParameters=true, module=SetVariables_#13}
   &ScaleSpriteSize(
    Scale Factor=&Lerp_Vector2(A=<1 4>, B=<1 12>, Alpha=$System.FN_Contrail_Setup.Dive Value)
    Scale Sprite Size Mode=::NewEnumerator2
   )
   &ScaleColor(
    Scale Alpha=&FloatFromCurve(
     FloatCurve=DataInterface() {propertyValues {Curve {keys [
         {interpMode=RCIM_Cubic, tangentMode=RCTM_User, time=0, value=0}
         {interpMode=RCIM_Cubic, tangentMode=RCTM_User, time=0.5, value=1}
         {tangentMode=RCTM_User, time=1, value=0}
        ]}}}
     Scale Curve=$System.FN_Contrail_Setup.Normalized Intro Age
    )
    ScaleA=true
    ScaleRGB=true
   )
  ]
  rendererClasses [&NiagaraSpriteRendererProperties, &NiagaraMeshRendererProperties]
  renderers [
   {rendererClass=&NiagaraSpriteRendererProperties}
   {rendererClass=&NiagaraMeshRendererProperties, rendererIndex=1}
  ]
 }
 Particles_SkeletalMesh {
  bEnabled=false
  emitterUpdateScript [
   &EmitterState()
   &SpawnRate(SpawnRate=40, Use Spawn Probability=false)
  ]
  particleSpawnScript [
   &PartitionParticles(Number Of Partitions=4)
   &InitializeParticle(
    Color Mode=::NewEnumerator0
    Lifetime Max=1
    Lifetime Mode=::NewEnumerator1
    Sprite Size=<5 20>
    Sprite Size Mode=::NewEnumerator4
    UsePositionOffset=false
    Write Material Random=true
   )
   &ShapeLocation(
    Apply To Particle Position=::NewEnumerator1
    MaskSpawnGroup=false
    Override Seed=false
    Shape Origin=&VectorFromFloat(Value=0)
    Sphere Radius=5
    UseOwnerScale=true
   )
  ]
  particleUpdateScript [
   &ParticleState()
   &SkeletalMeshLocation(
    Bone / Socket Index=$Particles.PartitionParticles.Partition
    MaskSpawnGroup=false
    Mesh Sampling Type=::NewEnumerator5
    Orientation Sampling=::NewEnumerator1
    Sampled Position Offset=$Particles.ShapeLocation.ShapePosition
    Skeletal Mesh=$System.Skeletal Mesh
    Socket Sampling Mode=::NewEnumerator5
   )
   /SetVariables(Particles.NewAge:NiagaraFloat=&Multiply_Float(A=$Particles.Age, B=0.75)) {isSetParameters=true}
   /SetVariables(Particles.Velocity=&Multiply_VectorByFloat(Vector=&Lerp_Vector(A=<0 0 0>, B=$System.FN_Contrail_Setup.Direction Velocity, Alpha=$Particles.NewAge), Float=1)) {isSetParameters=true}
   /SetVariables(Particles.Position=&AddVectorToPosition(
     Position=$Particles.Position
     Vector=&Multiply_VectorByFloat(Vector=$Particles.Velocity, Float=$Particles.Age)
    )) {isSetParameters=true, module=SetVariables_#13}
   /SetVariables(Particles.Velocity=$System.FN_Contrail_Setup.Direction Vector) {enabled=false, isSetParameters=true, module=SetVariables_#14}
   &ScaleColor(
    Scale Alpha=&FloatFromCurve(
     FloatCurve=DataInterface() {propertyValues {Curve {keys [
         {interpMode=RCIM_Cubic, tangentMode=RCTM_User, time=0, value=0}
         {interpMode=RCIM_Cubic, tangentMode=RCTM_User, time=0.5, value=1}
         {tangentMode=RCTM_User, time=1, value=0}
        ]}}}
     Scale Curve=$System.FN_Contrail_Setup.Normalized Intro Age
    )
    Scale RGB=<0.3 0.3 0.3>
    ScaleA=true
    ScaleRGB=true
   )
  ]
  rendererClasses [&NiagaraSpriteRendererProperties]
  renderers [{rendererClass=&NiagaraSpriteRendererProperties}]
 }
 ParticletoRibbon_SkeletalMesh {
  bEnabled=false
  emitterUpdateScript [
   &EmitterState()
   &SpawnBurst_Instantaneous(Spawn Count=40, Use Spawn Probability=false, UseLoopCountLimit=true)
  ]
  particleSpawnScript [
   &PartitionParticles(
    Normalized Index In Partition Includes One=true
    Number Of Partitions=4
    Particles Per Partition=10
    Partition Mode:ENiagara_PartitionMode=::NewEnumerator1
    Write Partition Index to Ribbon Link Order:ENiagara_AttributeSamplingApplyOutput=::NewEnumerator0
    Write Partition to Ribbon ID=::NewEnumerator0
   )
   &InitializeParticle(
    Color Mode=::NewEnumerator0
    Lifetime=0.5
    Ribbon Width=3
    Ribbon Width Mode=::NewEnumerator1
    Sprite Size Mode=::NewEnumerator3
    UsePositionOffset=false
    Write Material Random=true
   )
  ]
  particleUpdateScript [
   &ParticleState(Kill Particles When Lifetime Has Elapsed=false)
   &SkeletalMeshLocation(
    Bone / Socket Index=$Particles.PartitionParticles.Partition
    MaskSpawnGroup=false
    Mesh Sampling Type=::NewEnumerator5
    Skeletal Mesh=$System.Skeletal Mesh
    Socket Sampling Mode=::NewEnumerator5
   )
   /SetVariables(Particles.NewAge:NiagaraFloat=&Multiply_Float(A=$Particles.PartitionParticles.Normalized Index In Partition, B=$Particles.Initial.Lifetime)) {isSetParameters=true}
   /SetVariables(Particles.Velocity=&Multiply_VectorByFloat(
     Vector=&Lerp_Vector(A=$System.FN_Contrail_Setup.Direction Velocity Raw, B=$System.FN_Contrail_Setup.Direction Velocity, Alpha=$Particles.NewAge)
     Float=$Particles.Initial.Lifetime
    )) {isSetParameters=true}
   /SetVariables(Particles.Position=&AddVectorToPosition(
     Position=$Particles.Position
     Vector=&Multiply_VectorByFloat(Vector=$Particles.Velocity, Float=$Particles.PartitionParticles.Normalized Index In Partition)
    )) {isSetParameters=true, module=SetVariables_#13}
   &ScaleColor(Scale Alpha=$System.FN_Contrail_Setup.Normalized Intro Age, Scale RGB=<0.1 0.1 0.1>, ScaleA=true, ScaleRGB=true)
  ]
  rendererClasses [&NiagaraSpriteRendererProperties, &NiagaraRibbonRendererProperties]
  renderers [
   {rendererClass=&NiagaraSpriteRendererProperties}
   {rendererClass=&NiagaraRibbonRendererProperties, rendererIndex=1}
  ]
 }
 SimpleSpriteBurst {
  bEnabled=false
  emitterUpdateScript [
   &EmitterState(Life Cycle Mode=::NewEnumerator1, Loop Behavior=::NewEnumerator1, Loop Duration=2, UseLoopDelay=false)
   &SpawnBurst_Instantaneous(Use Spawn Probability=false, UseLoopCountLimit=false)
  ]
  particleSpawnScript [&InitializeParticle(
    Lifetime=2
    Position Offset=<0 0 200>
    Sprite Size Mode=::NewEnumerator3
    Uniform Sprite Size=50
    UsePositionOffset=true
    Write Material Random=true
   )]
  particleUpdateScript [
   &ParticleState(Kill Particles When Lifetime Has Elapsed=false)
   &ScaleColor(
    Scale Alpha=&FloatFromCurve(FloatCurve=DataInterface() {propertyValues {Curve {keys [<at -0.0046 1.0113 RCIM_Cubic>,<at 1.0011 0.002 RCIM_Cubic>]}}})
    ScaleA=true
    ScaleRGB=false
   ) {enabled=false}
   &SolveForcesAndVelocity(Clamp Velocity=false, Limit Acceleration=false)
  ]
  rendererClasses [&NiagaraSpriteRendererProperties]
  renderers [{rendererClass=&NiagaraSpriteRendererProperties}]
 }
}
summary {
 systemName=NS_Contrail_Cookies
 emitters [
  {
   bEnabled=false
   emitterName=Particle_Root
   rendererClasses [&NiagaraSpriteRendererProperties, &NiagaraMeshRendererProperties]
  }
  {bEnabled=false, emitterName=Particles_SkeletalMesh, rendererClasses [&NiagaraSpriteRendererProperties]}
  {bEnabled=false, emitterName=Gizmo_LocalSpace, rendererClasses [&NiagaraMeshRendererProperties]}
  {
   bEnabled=false
   emitterName=ParticletoRibbon_SkeletalMesh
   rendererClasses [&NiagaraSpriteRendererProperties, &NiagaraRibbonRendererProperties]
  }
  {emitterName=Bits_Trail, rendererClasses [&NiagaraMeshRendererProperties]}
  {bEnabled=false, emitterName=Bites_Burst001, rendererClasses [&NiagaraMeshRendererProperties]}
  {emitterName=Lines, rendererClasses [&NiagaraSpriteRendererProperties]}
  {bEnabled=false, emitterName=Cookies002, rendererClasses [&NiagaraMeshRendererProperties]}
  {emitterName=Cookies_NoLean, rendererClasses [&NiagaraMeshRendererProperties]}
  {bEnabled=false, emitterName=Cookies004, rendererClasses [&NiagaraMeshRendererProperties]}
  {bEnabled=false, emitterName=Bits_Trail001, rendererClasses [&NiagaraMeshRendererProperties]}
  {bEnabled=false, emitterName=Bites_Burst, rendererClasses [&NiagaraMeshRendererProperties]}
  {bEnabled=false, emitterName=Lines001, rendererClasses [&NiagaraSpriteRendererProperties]}
  {bEnabled=false, emitterName=Cookies, rendererClasses [&NiagaraMeshRendererProperties]}
  {bEnabled=false, emitterName=Cookies001, rendererClasses [&NiagaraMeshRendererProperties]}
  {emitterName=Cookies_Lean, rendererClasses [&NiagaraMeshRendererProperties]}
  {emitterName=Bits_Bursts, rendererClasses [&NiagaraMeshRendererProperties, &NiagaraSpriteRendererProperties]}
  {bEnabled=false, emitterName=Bites_Burst003, rendererClasses [&NiagaraMeshRendererProperties]}
  {bEnabled=false, emitterName=Default_Emote_TraversalDust, rendererClasses [&NiagaraSpriteRendererProperties]}
  {emitterName=Bits_Bursts001, rendererClasses [&NiagaraMeshRendererProperties]}
  {bEnabled=false, emitterName=SimpleSpriteBurst, rendererClasses [&NiagaraSpriteRendererProperties]}
 ]
 userVariables [
  User.bIsFrontEnd {type=NiagaraBool}
  User.bIsFrontEndPreview {type=NiagaraBool}
  User.OwnerLocalVelocity {type=Vector3f, description="", defaultValue {struct=Vector3f, value=<0 0 0>}}
  User.ParaGlideLeanAlpha {type=NiagaraFloat, defaultValue=0, description=""}
  User.RibbonIDOffset {type=NiagaraInt32}
 ]
}
